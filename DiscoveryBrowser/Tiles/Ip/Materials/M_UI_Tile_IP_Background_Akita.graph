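# UI material for the background of an Akita IP tile in the Discovery Browser.
# Two starfield samples, one flipped by negative scale, form a star layer alongside a screen-space-scaled circles texture and a procedural vignette from rotated UVs.
# Modular-button hover and focus transitions feed named reroutes, and interpolated values drive star and circle zoom.
# Chained lerps blend two color parameters and the layers, with a rounded-corner mask function also present.
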Material /DiscoveryBrowser/Tiles/Ip/Materials/M_UI_Tile_IP_Background_Akita
graph {
 properties {
  bAutomaticallySetUsageInEditor=false
  bUsedWithStaticMesh=false
  blendMode=BLEND_Translucent
  materialDomain=MD_UI
  usedShadingModels=Default Lit
 }
 nodes [
  MaterialFunctionCall(Corner Radius=Constant() {r=10}, Radius in Pixels=Constant() {r=1}) {materialFunction=&MF_UI_CoreUI_RoundedMask}
  n1 = MaterialFunctionCall() {materialFunction=/FortUILibrary/Systems/ModularButton/MaterialFunctions/MF_UIKit_ButtonStatesBase_Params}
  NamedRerouteDeclaration(None=@n1.Trans_Focused) {name=Is Focused, nodeColor=<rgba 0 0 0 1>}
  n2 = MaterialFunctionCall(UV=MaterialFunctionCall(UVs=MaterialFunctionCall() {materialFunction=&GetUserInterfaceUV}.Normalized UV) {materialFunction=&ScaleUVsByCenter}.UVs) {materialFunction=&MF_UI_UVScaling}
  n3 = LinearInterpolate(A=Constant() {r=1.15}, Alpha=LinearInterpolate(A=NamedRerouteUsage(), Alpha=NamedRerouteUsage())) {constA=1}
  n4 = Add(A=@n3, B=Constant() {r=2})
  n5 = TextureObject() {texture=/Game/Global/Textures/Default/PointSplatter/T_UI_Starfield}
  n6 = Multiply(
   A=TextureSample(
    UVs=MaterialFunctionCall(UVs=@n2.Scaled to Fit, Texture Scale=@n4) {materialFunction=&ScaleUVsByCenter}.UVs
    Tex=@n5
   ) {texture=&T_UI_Starfield}.RGB
   B=Constant() {r=1}
  ) {constB=2}
  n7 = OneMinus(None=Add(A=@n3, B=Constant() {r=4}))
  n8 = Multiply(
   A=TextureSample(
    UVs=MaterialFunctionCall(UVs=@n2.Scaled to Fit, Texture Scale=@n7) {materialFunction=&ScaleUVsByCenter}.UVs
    Tex=@n5
   ) {texture=&T_UI_Starfield}.RGB
   B=Constant() {r=1}
  ) {constB=2}
  NamedRerouteDeclaration(None=Saturate(None=Add(A=@n6, B=@n8))) {name=Stars, nodeColor=<rgba 1 0 0.7031 1>}
  n9 = MaterialFunctionCall() {materialFunction=&GetUserInterfaceUV}
  n10 = Constant() {r=0.04}
  n11 = ComponentMask(None=Add(A=MaterialFunctionCall(UVs=@n9.Normalized UV, "Rotation Angle (0-1)"=@n10) {materialFunction=&CustomRotator}.Rotated Values) {constB=-0.65}) {g=true}
  n12 = OneMinus(None=@n11)
  n13 = OneMinus(None=@n9.Normalized UV)
  n14 = Saturate(None=Multiply(A=DotProduct(A=@n13, B=@n9.Normalized UV)) {constB=2})
  n15 = Saturate(None=LinearInterpolate(
    B=Power(Base=@n14) {constExponent=0.3}
    Alpha=MaterialFunctionCall("Value (0-1)"=@n11) {materialFunction=&MF_UI_EaseCurves}.EaseOutCubic
   ) {constA=1})
  NamedRerouteDeclaration(None=Saturate(None=LinearInterpolate(B=@n12, Alpha=@n15))) {name=Vignette, nodeColor=<rgba 0 0 0 1>}
  NamedRerouteDeclaration(None=Multiply(A=ScalarParameter() {parameterName=Is Hovered Amount}, B=ScalarParameter() {defaultValue=1, parameterName=Hovered Speed})) {name=Is Hovered Amount, nodeColor=<rgba 0 0 0 1>}
  NamedRerouteDeclaration(None=@n1.Trans_Hovered) {name=Is Hovered, nodeColor=<rgba 0 0 0 1>}
  n16 = Constant() {r=1.125}
  n17 = LinearInterpolate(A=NamedRerouteUsage(), Alpha=NamedRerouteUsage())
  n18 = MaterialFunctionCall(Texture Scale=Add(A=LinearInterpolate(B=@n16, Alpha=@n17) {constA=1})) {materialFunction=&ScaleUVsByCenter}
  n19 = TextureProperty(Tex=NamedRerouteDeclaration(None=TextureObject() {samplerType=SAMPLERTYPE_LinearColor, texture=/DiscoveryBrowser/Art/Textures/T_UI_Akita_Background_Circles}) {name=Texture Object, nodeColor=<rgba 1 0 0.6328 1>})
  n20 = Constant2Vector() {g=1}
  n21 = TextureSample(
   UVs=Add(
    A=MaterialFunctionCall!(ScreenUV=@n18.UVs, TextureSize=@n19, ScalePivotPoint=@n20) {materialFunction=/FortUILibrary/MaterialFunctions/Basics/MF_UI_ScreenSpaceUVScaling}.Scaled to Fit
    B=Constant2Vector() {g=-0.25, r=-0.25}
   )
   Tex=NamedRerouteUsage()
  ) {samplerType=SAMPLERTYPE_LinearColor, texture=&T_UI_Akita_Background_Circles}
  n22 = Add(A=@n21.R, B=@n21.R)
  NamedRerouteDeclaration(None=Saturate(None=Add(A=@n22, B=@n21.B))) {name=Circles, nodeColor=<rgba 0 0 0 1>}
  n23 = LinearInterpolate(
   A=VectorParameter() {parameterName=Color A, defaultValue=<rgba 0.0227 0.028 0.035 1>}.RGB
   B=VectorParameter() {parameterName=Color B, defaultValue=<rgba 0.0058 0.0073 0.009 1>}.RGB
   Alpha=NamedRerouteUsage()
  )
  n24 = NamedRerouteUsage()
  n25 = Multiply(A=@n24) {constB=0.02}
  n26 = NamedRerouteUsage()
  Saturate(None=LinearInterpolate(A=LinearInterpolate(A=@n23, B=@n24, Alpha=@n25), B=@n26, Alpha=Multiply(A=@n26) {constB=0.65}))
 ]
}
