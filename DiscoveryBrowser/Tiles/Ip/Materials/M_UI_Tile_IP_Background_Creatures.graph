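# Translucent UI material that draws the orange background of the Creatures IP tile in the Discovery Browser.
# Two rotated, tiled layers with opposite horizontal offsets use random per-cell values to pick channels of the creature pattern texture, with sine-based line masks for rows.
# Radial vignette gradients, a rotated corner shape and a rounded-corner mask shape the tile.
# Button hover and focus states are exposed as reroutes, alongside a hover animation value.
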Material /DiscoveryBrowser/Tiles/Ip/Materials/M_UI_Tile_IP_Background_Creatures
graph {
 properties {
  bAutomaticallySetUsageInEditor=false
  bUsedWithStaticMesh=false
  blendMode=BLEND_Translucent
  materialDomain=MD_UI
  usedShadingModels=Default Lit
 }
 nodes [
  n1 = Constant3Vector() {constant=<rgba 1 0.4621 0.0284 1>}
  n2 = Constant3Vector() {constant=<rgba 0.8469 0.1274 0.0027 1>}
  n3 = NamedRerouteUsage()
  n4 = Constant3Vector() {constant=<rgba 0.8879 0.305 0.0152 1>}
  n5 = Constant3Vector() {constant=<rgba 0.8228 0.117 0.0021 1>}
  n6 = NamedRerouteUsage()
  LinearInterpolate(
   A=LinearInterpolate(A=LinearInterpolate(A=@n1, B=@n2, Alpha=@n3), B=LinearInterpolate(A=@n4, B=@n5, Alpha=@n6), Alpha=NamedRerouteUsage())
   Alpha=NamedRerouteUsage()
  ) {constA=1}
  MaterialFunctionCall(Corner Radius=Constant() {r=10}, Radius in Pixels=Constant() {r=1}) {materialFunction=&MF_UI_CoreUI_RoundedMask}
  n7 = MaterialFunctionCall() {materialFunction=/FortUILibrary/Systems/ModularButton/MaterialFunctions/MF_UIKit_ButtonStatesBase_Params}
  NamedRerouteDeclaration(None=@n7.Trans_Hovered) {name=Is Hovered, nodeColor=<rgba 0 0 0 1>}
  NamedRerouteDeclaration(None=@n7.Trans_Focused) {name=Is Focused, nodeColor=<rgba 0 0 0 1>}
  n8 = MaterialFunctionCall(UV=NamedRerouteDeclaration(None=MaterialFunctionCall() {materialFunction=&GetUserInterfaceUV}.Normalized UV) {name=Normalized UV, nodeColor=<rgba 0 0 0 1>}) {materialFunction=&MF_UI_UVScaling}
  NamedRerouteDeclaration(None=@n8.ScaleY) {name=Scale Y UV, nodeColor=<rgba 0 0 0 1>}
  NamedRerouteDeclaration(None=@n8.Scaled To Fill) {name=Scale to Fill, nodeColor=<rgba 0 0 0 1>}
  NamedRerouteDeclaration(None=@n8.Scaled to Fit) {name=Scale to Fit, nodeColor=<rgba 0 0 0 1>}
  n9 = NamedRerouteUsage()
  n10 = TextureSample(UVs=MaterialFunctionCall(UVs=NamedRerouteUsage(), Texture Scale=LinearInterpolate(Alpha=@n9) {constA=0.7, constB=0.5}) {materialFunction=&ScaleUVsByCenter}.UVs) {samplerType=SAMPLERTYPE_LinearColor, texture=/DiscoveryBrowser/Art/Textures/T_UI_Creatures_Pattern}
  n11 = MaterialFunctionCall(UVs=NamedRerouteUsage(), "Rotation Angle (0-1)"=ScalarParameter() {defaultValue=0.03, parameterName=Rotation Amount}) {materialFunction=&CustomRotator}
  n12 = Add(A=NamedRerouteUsage(), B=@n11.Rotated Values)
  n13 = Multiply(
   A=AppendVector(A=ComponentMask(None=@n12) {r=true}, B=ComponentMask(None=@n11.Rotated Values) {g=true})
   B=ScalarParameter() {defaultValue=3, parameterName=Pattern Size}
  ) {constB=6}
  n14 = MaterialFunctionCall!(UV=Floor(None=@n13), Seed=Constant() {r=0.2}) {materialFunction=&MF_UI_Random}
  n15 = Saturate(None=Multiply(A=NamedRerouteDeclaration(None=@n14.Result) {name=Random Cells, nodeColor=<rgba 0 1 0.4375 1>}, B=Constant() {r=1}) {constB=1.3})
  n16 = Floor(None=@n15)
  n17 = Min(A=OneMinus(None=@n16), B=Round(None=@n15))
  n18 = LinearInterpolate(A=LinearInterpolate(A=LinearInterpolate(A=@n10.G, B=@n10.G, Alpha=@n17), B=@n10.R, Alpha=@n16), Alpha=NamedRerouteUsage()) {constA=1, constB=0}
  n19 = NamedRerouteUsage()
  n20 = TextureSample(UVs=MaterialFunctionCall(UVs=NamedRerouteUsage(), Texture Scale=LinearInterpolate(Alpha=@n19) {constA=0.7, constB=0.5}) {materialFunction=&ScaleUVsByCenter}.UVs) {samplerType=SAMPLERTYPE_LinearColor, texture=&T_UI_Creatures_Pattern}
  n21 = MaterialFunctionCall(UVs=NamedRerouteUsage(), "Rotation Angle (0-1)"=ScalarParameter() {defaultValue=0.03, parameterName=Rotation Amount}) {materialFunction=&CustomRotator}
  n22 = Add(A=OneMinus(None=NamedRerouteUsage()), B=@n21.Rotated Values)
  n23 = Multiply(
   A=AppendVector(A=ComponentMask(None=@n22) {r=true}, B=ComponentMask(None=@n21.Rotated Values) {g=true})
   B=ScalarParameter() {defaultValue=3, parameterName=Pattern Size}
  ) {constB=6}
  n24 = MaterialFunctionCall(UV=Floor(None=@n23), Seed=Constant() {r=1.2}) {materialFunction=&MF_UI_Random}
  n25 = Saturate(None=Multiply(A=NamedRerouteDeclaration(None=@n24.Result) {name=Random Cells_1, nodeColor=<rgba 0 1 0.4375 1>}) {constB=1.3})
  n26 = Floor(None=@n25)
  n27 = Min(A=OneMinus(None=@n26), B=Round(None=@n25))
  n28 = LinearInterpolate(A=LinearInterpolate(A=LinearInterpolate(A=@n20.G, B=@n20.G, Alpha=@n27), B=@n20.R, Alpha=@n26), Alpha=NamedRerouteUsage()) {constB=0}
  NamedRerouteDeclaration(None=Saturate(None=Add(A=@n18, B=@n28))) {name=Pattern, nodeColor=<rgba 0 0 0 1>}
  NamedRerouteDeclaration(None=Frac(None=@n13)) {name=Pattern UV, nodeColor=<rgba 0.5234 1 0 1>}
  n29 = Divide(A=Multiply(A=ScalarParameter() {defaultValue=3, parameterName=Pattern Size}, B=@n21.Rotated Values) {constB=6})
  n30 = Sine(None=Frac(None=ComponentMask(None=@n29) {g=true}))
  NamedRerouteDeclaration(None=OneMinus(None=Ceil(None=@n30))) {name=Lines Mask B, nodeColor=<rgba 0 0 0 1>}
  n31 = Constant() {r=-0.025}
  n32 = LinearInterpolate(A=NamedRerouteUsage(), Alpha=NamedRerouteUsage())
  n33 = MaterialFunctionCall(
   UVs=Add(A=LinearInterpolate(B=@n31, Alpha=@n32), B=NamedRerouteUsage())
   Rotation Center=AppendVector(A=Constant() {r=0.7}, B=Constant() {r=0.5})
   "Rotation Angle (0-1)"=Constant() {r=0.38}
  ) {materialFunction=&CustomRotator}
  n34 = ComponentMask(None=@n33.Rotated Values) {r=true}
  n35 = SmoothStep(Value=@n34) {constMax=0.4925, constMin=0.5}
  n36 = ComponentMask(None=MaterialFunctionCall(UVs=@n33.Rotated Values, "Rotation Angle (0-1)"=Constant() {r=0.17}) {materialFunction=&CustomRotator}.Rotated Values) {g=true}
  n37 = LinearInterpolate(B=SmoothStep(Value=@n34) {constMax=0.5425, constMin=0.55}, Alpha=OneMinus(None=SmoothStep(Value=@n36) {constMax=0.5425, constMin=0.55}))
  NamedRerouteDeclaration(None=Saturate(None=Add(A=@n35, B=@n37))) {name=Corner Shape, nodeColor=<rgba 0 0 0 1>}
  n38 = NamedRerouteUsage()
  n39 = NamedRerouteUsage()
  NamedRerouteDeclaration(None=LinearInterpolate(B=Constant() {r=1.125}, Alpha=LinearInterpolate(A=@n38, Alpha=@n39)) {constA=1}) {name=Hover Animation, nodeColor=<rgba 1 0.9141 0 1>}
  n40 = Multiply(A=ScalarParameter() {defaultValue=3, parameterName=Pattern Size}, B=@n11.Rotated Values) {constB=6}
  n41 = Frac(None=ComponentMask(None=Divide(A=@n40)) {g=true})
  NamedRerouteDeclaration(None=Ceil(None=Sine(None=@n41))) {name=Lines Mask A, nodeColor=<rgba 0 0 0 1>}
  NamedRerouteDeclaration(None=@n8.ScaleX) {name=Scale X UV, nodeColor=<rgba 0 0 0 1>}
  n42 = ScalarParameter() {defaultValue=-0.2, group=Vignette, parameterName=Vig Mask X, sliderMax=1, sliderMin=-1}
  n43 = ScalarParameter() {defaultValue=0.03, group=Vignette, parameterName=Vig Max Y, sliderMax=1, sliderMin=-1}
  n44 = MaterialFunctionCall(
   UVs=Add(A=NamedRerouteUsage(), B=AppendVector(A=@n42, B=@n43))
   Texture Scale=AppendVector(
    A=ScalarParameter() {defaultValue=3, group=Vignette, parameterName=Vig Mask ScaleX, sliderMax=5}
    B=ScalarParameter() {defaultValue=1.5, group=Vignette, parameterName=Vig MaskAlpha ScaleY, sliderMax=5}
   )
  ) {materialFunction=&ScaleUVsByCenter}
  n45 = MaterialFunctionCall(UV=@n44.UVs, GradientStart=Constant() {r=1}, GradientEnd=Constant() {r=-0.2}) {materialFunction=&MF_UI_RadialGradient}
  NamedRerouteDeclaration(None=Saturate(None=MaterialFunctionCall("Value (0-1)"=@n45.Smooth) {materialFunction=&MF_UI_EaseCurves}.EaseInOutSine)) {name=Radial Gradient Background, nodeColor=<rgba 0 0 0 1>}
  NamedRerouteDeclaration(None=Frac(None=@n23)) {name=Pattern UV_1, nodeColor=<rgba 0.5234 1 0 1>}
  n46 = MaterialFunctionCall(UV=@n44.UVs, GradientStart=Constant() {r=0.8}, GradientEnd=Constant() {r=0.1}) {materialFunction=&MF_UI_RadialGradient}
  NamedRerouteDeclaration(None=Saturate(None=MaterialFunctionCall("Value (0-1)"=@n46.Smooth) {materialFunction=&MF_UI_EaseCurves}.EaseInSine)) {name=Radial Gradient Pattern, nodeColor=<rgba 0 0 0 1>}
 ]
}
