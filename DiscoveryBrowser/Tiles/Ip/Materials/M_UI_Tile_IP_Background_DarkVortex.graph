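# Translucent UI material that draws the Dark Vortex background for IP tiles in the Discovery browser.
# An eased, rotated UV gradient and a dark red palette form the base, with grunge texture modulated by squared blurred noise layered in.
# A vortex blob texture, offset and scaled about its center, provides a mask.
# Shared button-state parameters expose hover and focus values through named reroutes, alongside a hover amount and speed.
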Material /DiscoveryBrowser/Tiles/Ip/Materials/M_UI_Tile_IP_Background_DarkVortex
graph {
 properties {
  bAutomaticallySetUsageInEditor=false
  bUsedWithStaticMesh=false
  blendMode=BLEND_Translucent
  materialDomain=MD_UI
  usedShadingModels=Default Lit
 }
 nodes [
  MaterialFunctionCall(Corner Radius=Constant() {r=10}, Radius in Pixels=Constant() {r=1}) {materialFunction=&MF_UI_CoreUI_RoundedMask}
  n1 = MaterialFunctionCall() {materialFunction=/FortUILibrary/Systems/ModularButton/MaterialFunctions/MF_UIKit_ButtonStatesBase_Params}
  NamedRerouteDeclaration(None=@n1.Trans_Hovered) {name=Is Hovered, nodeColor=<rgba 0 0 0 1>}
  NamedRerouteDeclaration(None=@n1.Trans_Focused) {name=Is Focused, nodeColor=<rgba 0 0 0 1>}
  NamedRerouteDeclaration(None=Multiply(A=ScalarParameter() {parameterName=Is Hovered Amount}, B=ScalarParameter() {defaultValue=1, parameterName=Hovered Speed})) {name=Is Hovered Amount, nodeColor=<rgba 0 0 0 1>}
  n2 = TextureObject() {samplerType=SAMPLERTYPE_Alpha, texture=/DiscoveryBrowser/Tiles/Ip/Textures/T_UI_Discovery_Tile_Background_DarkVortex}
  n3 = Add(
   A=MaterialFunctionCall(
    UV=MaterialFunctionCall() {materialFunction=&GetUserInterfaceUV}.Normalized UV
    TextureSize=TextureProperty(Tex=@n2)
   ) {materialFunction=&MF_UI_UVScaling}.ScaleY
   B=Constant2Vector() {r=-0.2}
  )
  n4 = Constant() {r=1.4}
  n5 = LinearInterpolate(A=@n4, B=Add(A=@n4) {constB=0.125}, Alpha=LinearInterpolate(A=NamedRerouteUsage(), Alpha=NamedRerouteUsage())) {constA=1.7}
  NamedRerouteDeclaration(None=TextureSample(
    UVs=MaterialFunctionCall(UVs=@n3, Texture Scale=@n5) {materialFunction=&ScaleUVsByCenter}.UVs
    Tex=@n2
   ).R) {name=Blob Mask, nodeColor=<rgba 0 0 0 1>}
  n6 = MaterialFunctionCall(
   UVs=MaterialFunctionCall() {materialFunction=&GetUserInterfaceUV}.Normalized UV
   "Rotation Angle (0-1)"=Constant() {r=0.15}
  ) {materialFunction=&CustomRotator}
  NamedRerouteDeclaration(None=MaterialFunctionCall("Value (0-1)"=ComponentMask(None=@n6.Rotated Values) {g=true}) {materialFunction=&MF_UI_EaseCurves}.EaseInOutCubic) {name=Gradient Mask, nodeColor=<rgba 0 0 0 1>}
  Saturate(None=NamedRerouteUsage())
  n7 = MaterialFunctionCall(UV=MaterialFunctionCall() {materialFunction=&GetUserInterfaceUV}.Normalized UV) {materialFunction=&MF_UI_UVScaling}
  n8 = Constant() {r=0.25}
  n9 = TextureSample(
   UVs=@n7.Scaled to Fit
   Tex=TextureObject() {samplerType=SAMPLERTYPE_Grayscale, texture=&LowResBlurredNoise}
  ) {samplerType=SAMPLERTYPE_Grayscale}
  n10 = Multiply(
   A=TextureSample(
    UVs=MaterialFunctionCall(UVs=@n7.Scaled to Fit, Texture Scale=@n8) {materialFunction=&ScaleUVsByCenter}.UVs
    Tex=TextureObject() {samplerType=SAMPLERTYPE_Alpha, texture=/DiscoveryBrowser/Art/Textures/T_UI_DarkVortex_Background_Grunge}
   ) {texture=&T_UI_DarkVortex_Background_Grunge}.R
   B=LinearInterpolate(Alpha=Multiply(A=@n9.R, B=@n9.R)) {constA=0.1}
  )
  n11 = MaterialFunctionCall!(
   Base=LinearInterpolate(
    A=Constant3Vector() {constant=<rgba 0.4072 0.0009 0.016 1>}
    B=Constant3Vector() {constant=<rgba 0.0595 0.0015 0.003 1>}
    Alpha=NamedRerouteUsage()
   )
   Add=Saturate(None=Multiply(A=@n10) {constB=2})
  ) {materialFunction=/Engine/Functions/Engine_MaterialFunctions02/Math/MultiplyAdd}
  n12 = NamedRerouteUsage()
  Saturate(None=LinearInterpolate(A=Saturate(None=@n11.Result), Alpha=Multiply(A=@n12) {constB=0.8}) {constB=0})
 ]
}
