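# Translucent UI material for the Epic Games IP tile background in the Discovery browser, showing animated clouds and diagonal stripes over color gradients.
# A scrolling grayscale cloud texture is flow-map distorted by a rotating sample, with elliptical masks shaping three cloud sets.
# Stripes are anti-aliased distance fields from rotated UVs, combined through SDF union and subtraction with parameterized offsets and angles.
# Button-state hover and focus inputs are exposed, alongside a rounded-corner mask.
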Material /DiscoveryBrowser/Tiles/Ip/Materials/M_UI_Tile_IP_Background_EpicGames
graph {
 properties {
  bAutomaticallySetUsageInEditor=false
  bUsedWithStaticMesh=false
  blendMode=BLEND_Translucent
  materialDomain=MD_UI
  usedShadingModels=Default Lit
 }
 nodes [
  Constant() {r=0.75}
  LinearInterpolate(
   A=Constant3Vector() {constant=<rgba 0 0.2423 0.5583 1>}
   B=Constant3Vector() {constant=<rgba 0.0452 0.7605 0.8148 1>}
   Alpha=ComponentMask(None=MaterialFunctionCall() {materialFunction=&GetUserInterfaceUV}.Normalized UV) {g=true}
  )
  n1 = Constant3Vector() {constant=<rgba 0.8632 0.9047 0.9387 1>}
  n2 = Constant3Vector() {constant=<rgba 0.008 0.4287 0.7529 1>}
  n3 = Rotator(Coordinate=NamedRerouteUsage(), Time=Constant() {r=0.075}) {speed=6.2832}
  n4 = ComponentMask(None=Add(A=Multiply(A=@n3) {constB=0.5}) {constB=0.6}) {g=true}
  n5 = Constant3Vector() {constant=<rgba 0 0.159 0.7605 1>}
  n6 = Add(A=Rotator(Coordinate=NamedRerouteUsage(), Time=Constant() {r=0.4}) {speed=6.2832}) {constB=-0.2}
  LinearInterpolate(
   A=LinearInterpolate(A=NamedRerouteDeclaration(None=@n1) {name=Clouds Color, nodeColor=<rgba 0 0 0 1>}, B=@n2, Alpha=Saturate(None=@n4))
   B=@n5
   Alpha=Saturate(None=ComponentMask(None=@n6) {g=true})
  )
  n7 = NamedRerouteUsage()
  n8 = LinearInterpolate(A=LinearInterpolate(A=@n5, B=@n2, Alpha=ComponentMask(None=@n7) {g=true}), B=NamedRerouteUsage(), Alpha=NamedRerouteUsage()) {constAlpha=0}
  n9 = Constant3Vector() {constant=<rgba 0.6376 0.6376 0.6376 1>}
  n10 = Constant3Vector() {constant=<rgba 1 1 1 0>}
  n11 = Saturate(None=ComponentMask(None=Add(A=@n6) {constB=0.375}) {r=true})
  n12 = ComponentMask(None=NamedRerouteUsage()) {r=true}
  LinearInterpolate(
   A=LinearInterpolate(
    A=NamedRerouteDeclaration(None=@n8) {name=Clouds Visuals, nodeColor=<rgba 0 0 0 1>}
    B=LinearInterpolate(A=@n9, B=@n10, Alpha=@n11)
    Alpha=NamedRerouteUsage()
   )
   B=LinearInterpolate(
    A=Constant3Vector() {constant=<rgba 0.007 0.007 0.007 1>}
    B=Constant3Vector() {constant=<rgba 0.0222 0.0212 0.0222 1>}
    Alpha=Saturate(None=@n12)
   ) {constA=0.007, constB=0.0222}
   Alpha=NamedRerouteUsage()
  )
  MaterialFunctionCall(Corner Radius=Constant() {r=10}, Radius in Pixels=Constant() {r=1}) {materialFunction=&MF_UI_CoreUI_RoundedMask}
  n13 = Multiply(A=NamedRerouteUsage()) {constB=-1}
  n14 = Add(A=NamedRerouteUsage(), B=Constant2Vector() {g=-0.5, r=-0.5})
  n15 = Constant2Vector() {g=0.6666, r=1.2}
  n16 = ConstantBiasScale(None=MaterialFunctionCall(UV=MaterialFunctionCall(UVs=@n14, Texture Scale=@n15) {materialFunction=&ScaleUVsByCenter}.UVs) {materialFunction=&MF_UI_UVScaling}.ScaleY) {bias=-0.5, scale=2}
  n17 = SmoothStep(Min=@n13, Value=OneMinus(None=Length(None=@n16) {is=class MaterialExpressionLength}))
  n18 = Multiply(A=@n17, B=NamedRerouteUsage())
  Multiply(A=@n18, B=@n18)
  n19 = MaterialFunctionCall() {materialFunction=/FortUILibrary/Systems/ModularButton/MaterialFunctions/MF_UIKit_ButtonStatesBase_Params}
  NamedRerouteDeclaration(None=@n19.Trans_Hovered) {name=Is Hovered, nodeColor=<rgba 0 0 0 1>}
  NamedRerouteDeclaration(None=@n19.Trans_Focused) {name=Is Focused, nodeColor=<rgba 0 0 0 1>}
  n20 = NamedRerouteUsage()
  n21 = NamedRerouteUsage()
  NamedRerouteDeclaration(None=LinearInterpolate(A=LinearInterpolate(A=@n20, Alpha=@n21), Alpha=NamedRerouteUsage())) {name=Masked Clouds, nodeColor=<rgba 0 0 0 1>}
  n22 = MaterialFunctionCall(UV=NamedRerouteDeclaration(None=MaterialFunctionCall() {materialFunction=&GetUserInterfaceUV}.Normalized UV) {name=UV, nodeColor=<rgba 0 0 0 1>}) {materialFunction=&MF_UI_UVScaling}
  NamedRerouteDeclaration(None=@n22.Scaled to Fit) {name=Scale Fit UV, nodeColor=<rgba 0 0 0 1>}
  n23 = NamedRerouteUsage()
  n24 = Constant2Vector() {g=-0.5, r=0.5}
  n25 = MaterialFunctionCall(UV=MaterialFunctionCall(UVs=Add(A=@n23, B=@n24), Texture Scale=Constant2Vector() {g=1.2, r=2}) {materialFunction=&ScaleUVsByCenter}.UVs) {materialFunction=&MF_UI_UVScaling}
  n26 = OneMinus(None=Length(None=ConstantBiasScale(None=@n25.ScaleY) {bias=-0.5, scale=2}))
  n27 = Power(Base=SmoothStep(Min=@n13, Value=@n26)) {constExponent=1.375}
  n28 = NamedRerouteUsage()
  NamedRerouteDeclaration(None=LinearInterpolate(A=Multiply(A=@n27, B=@n28), B=@n27, Alpha=Saturate(None=@n26))) {name=Cloud Set 01, nodeColor=<rgba 0 0 0 1>}
  NamedRerouteDeclaration(None=@n18) {name=Cloud Set 02, nodeColor=<rgba 0 0 0 1>}
  NamedRerouteDeclaration(None=Multiply(A=ScalarParameter() {parameterName=Is Hovered Amount}, B=ScalarParameter() {defaultValue=1, parameterName=Hovered Speed})) {name=Is Hovered Amount, nodeColor=<rgba 0 0 0 1>}
  n29 = NamedRerouteUsage()
  n30 = Divide(A=Subtract(A=Multiply(A=@n29) {constB=1.05}) {constB=0.5}, B=LinearInterpolate(Alpha=NamedRerouteUsage()) {constA=1, constB=1.125})
  n31 = Time()
  n32 = Multiply(A=@n31) {constB=0.01}
  n33 = Multiply(A=@n31) {constB=0.001}
  n34 = Multiply(A=@n29) {constB=0.6666}
  n35 = Multiply(A=Time()) {constB=-0.005}
  n36 = TextureSample(UVs=Add(A=Rotator(Coordinate=@n34, Time=@n35) {speed=6.2832}, B=@n33), Tex=NamedRerouteUsage())
  n37 = MaterialFunctionCall!(
   UVs=Add(A=Add(A=@n30) {constB=0.5}, B=AppendVector(A=@n32, B=@n33))
   Texture Object=NamedRerouteDeclaration(None=TextureObject() {samplerType=SAMPLERTYPE_LinearGrayscale, texture=/DiscoveryBrowser/Tiles/Ip/Textures/T_UI_Discovery_Tile_Background_EpicGames}) {name=Clouds Texture, nodeColor=<rgba 0 0 0 1>}
   Grayscale Flow Map=Multiply(A=@n36.R, B=@n36.R)
   Flow Direction=Constant() {r=0.5648}
   Flow Intensity=Constant() {r=0.07}
   Flow Speed=Constant() {r=0.125}
  ) {materialFunction=/FortUILibrary/MaterialFunctions/Effects/MF_UI_GrayscaleFlowMap_New}
  NamedRerouteDeclaration(None=SmoothStep(Value=ComponentMask(None=@n37.Distorted Texture) {r=true})) {name=Clouds Pattern, nodeColor=<rgba 0 0 0 1>}
  n38 = NamedRerouteUsage()
  n39 = Constant2Vector() {g=0.5, r=-0.5}
  n40 = MaterialFunctionCall(UV=MaterialFunctionCall(UVs=Add(A=@n38, B=@n39), Texture Scale=Constant2Vector() {g=0.5, r=0.75}) {materialFunction=&ScaleUVsByCenter}.UVs) {materialFunction=&MF_UI_UVScaling}
  n41 = OneMinus(None=Length(None=ConstantBiasScale(None=@n40.ScaleY) {bias=-0.5, scale=2}))
  n42 = SmoothStep(Min=@n13, Value=@n41)
  n43 = LinearInterpolate(A=Multiply(A=@n42, B=NamedRerouteUsage()), B=@n42, Alpha=Saturate(None=@n41))
  NamedRerouteDeclaration(None=Multiply(A=@n43, B=@n43)) {name=Cloud Set 03, nodeColor=<rgba 0 0 0 1>}
  NamedRerouteDeclaration(None=LinearInterpolate(A=NamedRerouteUsage(), Alpha=NamedRerouteUsage())) {name=Interaction, nodeColor=<rgba 0 0 0 1>}
  NamedRerouteDeclaration(None=@n22.Scaled To Fill) {name=Scale Fill UV, nodeColor=<rgba 0 0 0 1>}
  NamedRerouteDeclaration(None=@n22.ScaleY) {name=Scale Y UV, nodeColor=<rgba 0 0 0 1>}
  n44 = ScalarParameter() {defaultValue=0.476, group=Stripes, parameterName=Stripe Offset}
  n45 = Add(A=@n44) {constB=-0.125}
  n46 = NamedRerouteUsage()
  n47 = Subtract(
   A=Rotator(Coordinate=NamedRerouteUsage(), Time=ScalarParameter() {defaultValue=0.12, group=Stripes, parameterName=Stripe Rotation}) {speed=6.2832}
   B=AppendVector(A=LinearInterpolate(A=@n44, B=@n45, Alpha=@n46) {constA=1, constB=1.125}, B=Constant())
  )
  NamedRerouteDeclaration(None=MaterialFunctionCall(DistanceField=ComponentMask(None=@n47) {r=true}) {materialFunction=&MF_UI_AntiAliasing}.Inner) {name=Divider Stripe Mask, nodeColor=<rgba 0 0 0 1>}
  n48 = Subtract(
   A=Rotator(Coordinate=@n47, Time=ScalarParameter() {defaultValue=0.01, group=Stripes, parameterName=Small Stripe Rotation}) {speed=6.2832}
   B=ScalarParameter() {defaultValue=0.1, group=Stripes, parameterName=Small Stripe Offset}
  ) {constB=0.1}
  n49 = ComponentMask(None=@n48) {r=true}
  n50 = Rotator(Coordinate=@n48, Time=ScalarParameter() {group=Stripes, parameterName=Triangle 01 Rotation}) {speed=6.2832}
  n51 = AppendVector(A=ScalarParameter() {defaultValue=0.132, group=Stripes, parameterName=Triangle 01 Offset}, B=Constant())
  n52 = Rotator(Coordinate=@n48, Time=ScalarParameter() {defaultValue=0.0175, group=Stripes, parameterName=Triangle 02 Rotation}) {speed=6.2832}
  n53 = AppendVector(A=ScalarParameter() {defaultValue=0.06, group=Stripes, parameterName=Triangle 02 Offset}, B=Constant())
  n54 = MaterialFunctionCall!(
   DistanceFieldA=ComponentMask(None=Subtract(A=@n50, B=@n51)) {r=true}
   DistanceFieldB=ComponentMask(None=Subtract(A=@n52, B=@n53)) {r=true}
   BlendAmount
  ) {materialFunction=/FortUILibrary/MaterialFunctions/Shapes/Modifiers/MF_UI_SDF_Subtraction}
  NamedRerouteDeclaration(None=MaterialFunctionCall(DistanceField=MaterialFunctionCall!(DistanceFieldA=@n49, DistanceFieldB=@n54.Exact, BlendAmount) {materialFunction=/FortUILibrary/MaterialFunctions/Shapes/Modifiers/MF_UI_SDF_Union}.Exact) {materialFunction=&MF_UI_AntiAliasing}.Inner) {name=Small Stripe Mask, nodeColor=<rgba 0 0 0 1>}
  NamedRerouteUsage()
  Power(Base=@n17) {constExponent=1.1}
  Saturate(None=NamedRerouteUsage())
 ]
}
