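# UI material for the Fall Guys IP tile background in the Discovery browser, translucent with a rounded-corner mask.
# An alpha pattern texture scales around its center using button hover and focus states.
# A rotated, cubic-eased UV gradient forms a vignette.
# Two blue color parameters are blended and given a subtle additive term to build the background from these gradient and pattern layers.
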Material /DiscoveryBrowser/Tiles/Ip/Materials/M_UI_Tile_IP_Background_FallGuys
graph {
 properties {
  bAutomaticallySetUsageInEditor=false
  bUsedWithStaticMesh=false
  blendMode=BLEND_Translucent
  materialDomain=MD_UI
  usedShadingModels=Default Lit
 }
 nodes [
  MaterialFunctionCall(Corner Radius=Constant() {r=10}, Radius in Pixels=Constant() {r=1}) {materialFunction=&MF_UI_CoreUI_RoundedMask}
  n1 = MaterialFunctionCall() {materialFunction=/FortUILibrary/Systems/ModularButton/MaterialFunctions/MF_UIKit_ButtonStatesBase_Params}
  NamedRerouteDeclaration(None=@n1.Trans_Hovered) {name=Is Hovered, nodeColor=<rgba 0 0 0 1>}
  NamedRerouteDeclaration(None=@n1.Trans_Focused) {name=Is Focused, nodeColor=<rgba 0 0 0 1>}
  NamedRerouteDeclaration(None=Multiply(A=ScalarParameter() {parameterName=Is Hovered Amount}, B=ScalarParameter() {defaultValue=1, parameterName=Hovered Speed})) {name=Is Hovered Amount, nodeColor=<rgba 0 0 0 1>}
  n2 = MaterialFunctionCall() {materialFunction=&GetUserInterfaceUV}
  n3 = LinearInterpolate(B=Constant() {r=1.125}, Alpha=LinearInterpolate(A=NamedRerouteUsage(), Alpha=NamedRerouteUsage())) {constA=1}
  NamedRerouteDeclaration(None=TextureSample(
    UVs=MaterialFunctionCall(UVs=@n2.Normalized UV, Texture Scale=@n3) {materialFunction=&ScaleUVsByCenter}.UVs
    Tex=TextureObject() {samplerType=SAMPLERTYPE_Alpha, texture=/DiscoveryBrowser/Tiles/Ip/Textures/T_UI_Discovery_Tile_Background_FallGuys}
   ) {samplerSource=SSM_Wrap_WorldGroupSettings}.R) {name=Pattern, nodeColor=<rgba 0 0 0 1>}
  n4 = MaterialFunctionCall(
   UVs=MaterialFunctionCall() {materialFunction=&GetUserInterfaceUV}.Normalized UV
   "Rotation Angle (0-1)"=Constant() {r=0.15}
  ) {materialFunction=&CustomRotator}
  NamedRerouteDeclaration(None=MaterialFunctionCall("Value (0-1)"=ComponentMask(None=@n4.Rotated Values) {g=true}) {materialFunction=&MF_UI_EaseCurves}.EaseInOutCubic) {name=Vignette, nodeColor=<rgba 0 0 0 1>}
  Saturate(None=NamedRerouteUsage())
  n5 = VectorParameter() {parameterName=Color A, defaultValue=<rgba 0 0.1248 0.5906 1>}
  n6 = VectorParameter() {parameterName=Color B, defaultValue=<rgba 0.0331 0.5841 0.9301 1>}
  n7 = NamedRerouteUsage()
  n8 = NamedRerouteUsage()
  Saturate(None=MaterialFunctionCall!(Base=LinearInterpolate(A=@n5.RGB, B=@n6.RGB, Alpha=@n7), Add=Multiply(A=@n8) {constB=0.15}) {materialFunction=/Engine/Functions/Engine_MaterialFunctions02/Math/MultiplyAdd}.Result)
 ]
}
