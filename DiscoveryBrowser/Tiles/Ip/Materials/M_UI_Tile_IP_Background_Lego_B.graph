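# Translucent UI material for the Lego IP tile background in the Discovery browser, variant B.
# Two offset, rotated rounded-mask shapes form front and back layers, with UV scaling that appears driven by button hover and focus states.
# Procedural halftone masks tile circles on rotated UV grids, using per-cell random sizes for large dots and stepped gradient falloff for small ones.
# Nested lerps blend three orange-to-yellow color parameters through these masks.
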Material /DiscoveryBrowser/Tiles/Ip/Materials/M_UI_Tile_IP_Background_Lego_B
graph {
 properties {
  bAutomaticallySetUsageInEditor=false
  bUsedWithStaticMesh=false
  blendMode=BLEND_Translucent
  materialDomain=MD_UI
  usedShadingModels=Default Lit
 }
 nodes [
  n1 = NamedRerouteUsage()
  n2 = VectorParameter() {parameterName=Color B, defaultValue=<rgba 0.9911 0.4452 0.0003 1>}
  n3 = Multiply(A=NamedRerouteUsage()) {constB=0.075}
  n4 = NamedRerouteUsage()
  n5 = NamedRerouteUsage()
  LinearInterpolate(
   A=LinearInterpolate(
    A=VectorParameter() {parameterName=Color A, defaultValue=<rgba 0.9911 0.2789 0.0123 1>}.RGB
    B=Constant()
    Alpha=Multiply(A=@n1) {constB=0.1}
   )
   B=LinearInterpolate(
    A=VectorParameter() {parameterName=Color C, defaultValue=<rgba 0.9879 1 0 1>}.RGB
    B=LinearInterpolate(A=@n2.RGB, Alpha=@n3) {constB=0}
    Alpha=@n4
   )
   Alpha=Saturate(None=Add(A=@n5, B=@n4))
  )
  MaterialFunctionCall(Corner Radius=Constant() {r=10}, Radius in Pixels=Constant() {r=1}) {materialFunction=&MF_UI_CoreUI_RoundedMask}
  n6 = MaterialFunctionCall() {materialFunction=/FortUILibrary/Systems/ModularButton/MaterialFunctions/MF_UIKit_ButtonStatesBase_Params}
  NamedRerouteDeclaration(None=@n6.Trans_Hovered) {name=Is Hovered, nodeColor=<rgba 0 0 0 1>}
  NamedRerouteDeclaration(None=@n6.Trans_Focused) {name=Is Focused, nodeColor=<rgba 0 0 0 1>}
  n7 = MaterialFunctionCall(UV=NamedRerouteDeclaration(None=MaterialFunctionCall() {materialFunction=&GetUserInterfaceUV}.Normalized UV) {name=Normalized UV, nodeColor=<rgba 0 0 0 1>}) {materialFunction=&MF_UI_UVScaling}
  NamedRerouteDeclaration(None=@n7.ScaleY) {name=Scale Y UV, nodeColor=<rgba 0 0 0 1>}
  NamedRerouteDeclaration(None=@n7.Scaled To Fill) {name=Scale to Fill, nodeColor=<rgba 0 0 0 1>}
  n8 = Multiply(A=Rotator(Coordinate=NamedRerouteUsage(), Time=Constant() {r=0.4}) {speed=6.2832}, B=Constant() {desc=Halftone Amount, r=10}) {constB=32}
  n9 = Frac(None=@n8)
  n10 = MaterialFunctionCall!(UV=Floor(None=@n8), Seed) {materialFunction=&MF_UI_Random}
  n11 = LinearInterpolate(A=Constant() {desc=Halftone Min, r=0.35}, B=Constant() {desc=Halftone Max, r=0.7}, Alpha=Saturate(None=Max(A=@n10.Result) {constB=0})) {constA=0.25, constB=0.75}
  NamedRerouteDeclaration(None=Saturate(None=MaterialFunctionCall(UVs=@n9, Size=@n11) {materialFunction=&MF_UI_Circle}.Fill)) {name=Halftone Mask Big, nodeColor=<rgba 0 0 0 1>}
  NamedRerouteDeclaration(None=@n7.Scaled to Fit) {name=Scale to Fit, nodeColor=<rgba 0 0 0 1>}
  n12 = NamedRerouteUsage()
  n13 = NamedRerouteUsage()
  n14 = MaterialFunctionCall(UVs=NamedRerouteUsage(), Texture Scale=LinearInterpolate(B=Constant() {r=1.125}, Alpha=LinearInterpolate(A=@n12, Alpha=@n13)) {constA=1}) {materialFunction=&ScaleUVsByCenter}
  n15 = MaterialFunctionCall(UVs=@n14.UVs, "Rotation Angle (0-1)"=Constant() {r=0.01}) {materialFunction=&CustomRotator}
  n16 = MaterialFunctionCall(UVs=@n14.UVs, "Rotation Angle (0-1)"=Constant() {r=0.06}) {materialFunction=&CustomRotator}
  n17 = AppendVector(
   A=ComponentMask(None=Add(A=@n15.Rotated Values) {constB=-0.45}) {r=true}
   B=ComponentMask(None=Add(A=@n16.Rotated Values) {constB=0.1}) {g=true}
  )
  n18 = Constant() {r=1}
  n19 = Constant() {r=10}
  n20 = Constant() {r=1}
  NamedRerouteDeclaration(None=Saturate(None=MaterialFunctionCall(UV=@n17, Size=@n18, Corner Radius=@n19, Radius in Pixels=@n20) {materialFunction=&MF_UI_CoreUI_RoundedMask}.Shape Mask)) {name=Shape Front, nodeColor=<rgba 0 0 0 1>}
  NamedRerouteDeclaration(None=@n7.ScaleX) {name=Scale X UV, nodeColor=<rgba 0 0 0 1>}
  n21 = MaterialFunctionCall(UVs=@n14.UVs, "Rotation Angle (0-1)"=Constant() {r=0.005}) {materialFunction=&CustomRotator}
  n22 = MaterialFunctionCall(UVs=@n14.UVs, "Rotation Angle (0-1)"=Constant() {r=0.014}) {materialFunction=&CustomRotator}
  n23 = AppendVector(
   A=ComponentMask(None=Add(A=@n21.Rotated Values) {constB=-0.435}) {r=true}
   B=ComponentMask(None=Add(A=@n22.Rotated Values) {constB=0.16}) {g=true}
  )
  n24 = Constant() {r=1}
  n25 = Constant() {r=10}
  n26 = Constant() {r=1}
  NamedRerouteDeclaration(None=Saturate(None=MaterialFunctionCall(UV=@n23, Size=@n24, Corner Radius=@n25, Radius in Pixels=@n26) {materialFunction=&MF_UI_CoreUI_RoundedMask}.Shape Mask)) {name=Shape Back, nodeColor=<rgba 0 0 0 1>}
  n27 = NamedRerouteUsage()
  n28 = Rotator(Coordinate=@n27, Time=Constant() {r=0.4}) {speed=6.2832}
  n29 = Constant() {desc=Halftone Amount, r=60}
  n30 = Multiply(A=@n28, B=@n29) {constB=32}
  n31 = Constant() {desc=Halftone Min, r=0.45}
  n32 = Constant() {desc=Halftone Max, r=0.45}
  n33 = Floor(None=@n30)
  n34 = Saturate(None=Max(A=MaterialFunctionCall(UV=@n33) {materialFunction=&MF_UI_Random}.Result) {constB=0})
  n35 = NamedRerouteDeclaration(None=Constant() {r=0.1}) {name=Halftone Rotation, nodeColor=<rgba 0 0 0 1>}
  n36 = MaterialFunctionCall!("Value (0-1)"=ComponentMask(None=Rotator(Coordinate=@n27, Time=@n35) {speed=6.2832}) {r=true}, Steps=@n29) {materialFunction=/FortUILibrary/MaterialFunctions/Basics/MF_UI_StepRatio}
  n37 = Saturate(None=MaterialFunctionCall(Input=OneMinus(None=@n36.Stepped Gradient), Input Low=Constant(), Input High=Constant() {r=0.5}) {materialFunction=&RemapValue_0-1}.Result)
  n38 = Length(None=Subtract(A=@n28, B=Constant2Vector() {g=1, r=0.5}) {constB=0.5}) {is=class MaterialExpressionLength}
  n39 = Saturate(None=MaterialFunctionCall(Input=Saturate(None=@n38), Input Low=Constant() {r=0.5}, Input High=Constant() {r=0.7}) {materialFunction=&RemapValue_0-1}.Result)
  n40 = MaterialFunctionCall(
   UVs=Frac(None=@n30)
   Size=Multiply(A=LinearInterpolate(A=@n31, B=@n32, Alpha=@n34) {constA=0.25, constB=0.75}, B=LinearInterpolate(B=@n37, Alpha=@n39) {constA=1})
  ) {materialFunction=&MF_UI_Circle}
  n41 = Constant()
  n42 = OneMinus(None=@n39)
  NamedRerouteDeclaration(None=Saturate(None=LinearInterpolate(A=@n40.Fill, B=@n41, Alpha=@n42))) {name=Halftone Mask Small, nodeColor=<rgba 0 0 0 1>}
 ]
}
