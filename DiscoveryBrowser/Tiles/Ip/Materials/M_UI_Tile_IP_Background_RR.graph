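# Translucent UI material for the Rocket Racing IP tile background in the Discovery browser, with button-state inputs for hover and focus.
# A flag pattern texture is mirrored by blending left and right screen-space-scaled UVs at the horizontal midpoint, with clipped edge pixels and animatable center scaling.
# An eased procedural vignette is built alongside it, and named reroutes feed a two-color blend with a multiply-add highlight.
# A rounded-corner mask function is also present.
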Material /DiscoveryBrowser/Tiles/Ip/Materials/M_UI_Tile_IP_Background_RR
graph {
 properties {
  bAutomaticallySetUsageInEditor=false
  bUsedWithStaticMesh=false
  blendMode=BLEND_Translucent
  materialDomain=MD_UI
  usedShadingModels=Default Lit
 }
 nodes [
  MaterialFunctionCall(Corner Radius=Constant() {r=10}, Radius in Pixels=Constant() {r=1}) {materialFunction=&MF_UI_CoreUI_RoundedMask}
  n1 = MaterialFunctionCall() {materialFunction=/FortUILibrary/Systems/ModularButton/MaterialFunctions/MF_UIKit_ButtonStatesBase_Params}
  NamedRerouteDeclaration(None=@n1.Trans_Focused) {name=Is Focused, nodeColor=<rgba 0 0 0 1>}
  n2 = MaterialFunctionCall() {materialFunction=&GetUserInterfaceUV}
  n3 = NamedRerouteUsage()
  n4 = NamedRerouteUsage()
  n5 = MaterialFunctionCall(UVs=@n2.Normalized UV, Texture Scale=LinearInterpolate(B=Constant() {r=1.125}, Alpha=LinearInterpolate(A=@n3, Alpha=@n4)) {constA=1}) {materialFunction=&ScaleUVsByCenter}
  n6 = TextureProperty(Tex=NamedRerouteDeclaration(None=TextureObject() {samplerType=SAMPLERTYPE_Alpha, texture=/DiscoveryBrowser/Art/Textures/T_UI_Discovery_RocketRacing_Pattern}) {name=Texture Object, nodeColor=<rgba 0 0 0 1>})
  n7 = Add(
   A=MaterialFunctionCall!(ScreenUV=@n5.UVs, TextureSize=@n6, ScalePivotPoint=Constant2Vector() {g=1}) {materialFunction=/FortUILibrary/MaterialFunctions/Basics/MF_UI_ScreenSpaceUVScaling}.ScaleX
   B=Constant2Vector() {g=-0.25, r=0.5}
  ) {desc=Left Side}
  n8 = Add(
   A=MaterialFunctionCall(ScreenUV=OneMinus(None=@n5.UVs), TextureSize=@n6, ScalePivotPoint=Constant2Vector() {g=1}) {materialFunction=&MF_UI_ScreenSpaceUVScaling}.ScaleX
   B=Constant2Vector() {g=-0.05, r=0.7}
  ) {desc=Right Side}
  n9 = Round(None=ComponentMask(None=@n2.Normalized UV) {r=true})
  n10 = TextureSample(
   UVs=NamedRerouteDeclaration(None=LinearInterpolate(A=@n7, B=@n8, Alpha=@n9)) {name=Mirror Flag UV, nodeColor=<rgba 0 0 0 1>}
   Tex=NamedRerouteUsage()
  ) {
   samplerSource=SSM_Clamp_WorldGroupSettings
   samplerType=SAMPLERTYPE_Alpha
   texture=&T_UI_Discovery_RocketRacing_Pattern
  }
  n11 = NamedRerouteUsage()
  NamedRerouteDeclaration(None=Multiply(
    A=@n10.R
    B=MaterialFunctionCall!(UVs=@n11, Alpha=@n10.A) {materialFunction=/FortUILibrary/MaterialFunctions/Basics/MF_UI_ClipEdgePixel}.Result
   )) {name=Pattern, nodeColor=<rgba 0 0 0 1>}
  n12 = MaterialFunctionCall() {materialFunction=&GetUserInterfaceUV}
  n13 = ComponentMask(None=@n12.Normalized UV) {g=true}
  n14 = Multiply(A=DotProduct(A=OneMinus(None=@n12.Normalized UV), B=@n12.Normalized UV)) {constB=2}
  n15 = LinearInterpolate(
   B=Power(Base=Saturate(None=@n14))
   Alpha=MaterialFunctionCall("Value (0-1)"=@n13) {materialFunction=&MF_UI_EaseCurves}.EaseOutCubic
  ) {constA=1}
  NamedRerouteDeclaration(None=LinearInterpolate(B=OneMinus(None=@n13), Alpha=Saturate(None=@n15))) {name=Vignette, nodeColor=<rgba 0 0 0 1>}
  NamedRerouteDeclaration(None=Multiply(A=ScalarParameter() {parameterName=Is Hovered Amount}, B=ScalarParameter() {defaultValue=1, parameterName=Hovered Speed})) {name=Is Hovered Amount, nodeColor=<rgba 0 0 0 1>}
  NamedRerouteDeclaration(None=@n1.Trans_Hovered) {name=Is Hovered, nodeColor=<rgba 0 0 0 1>}
  n16 = VectorParameter() {parameterName=Color A, defaultValue=<rgba 0.0194 0.006 0.0885 1>}
  n17 = VectorParameter() {parameterName=Color B, defaultValue=<rgba 0.0482 0.0065 0.521 1>}
  n18 = NamedRerouteUsage()
  n19 = NamedRerouteUsage()
  Saturate(None=MaterialFunctionCall!(Base=LinearInterpolate(A=@n16.RGB, B=@n17.RGB, Alpha=@n18), Add=Multiply(A=@n19) {constB=0.35}) {materialFunction=/Engine/Functions/Engine_MaterialFunctions02/Math/MultiplyAdd}.Result)
 ]
}
