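# Translucent UI material for the animated "Refrigerator" IP tile background in the Discovery browser.
# Two procedural "Honmoon" line layers, built from wave-warped, rotated, repeated UVs and blurred-texture distortion maps, are masked by gradients.
# These join eased radial glows, a flare texture and a time-animated circle ring broken up by panning noise, over navy, purple, blue and teal colours.
# Shared button-state parameters supply hover and focus transitions, with named reroutes linking the layers.
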Material /DiscoveryBrowser/Tiles/Ip/Materials/M_UI_Tile_IP_Background_Refrigerator
graph {
 properties {
  bAutomaticallySetUsageInEditor=false
  bUsedWithStaticMesh=false
  blendMode=BLEND_Translucent
  materialDomain=MD_UI
  usedShadingModels=Default Lit
 }
 nodes [
  Constant() {r=0.108}
  Constant() {r=-0.5}
  Constant() {r=0.25}
  MaterialFunctionCall(Corner Radius=Constant() {r=10}, Radius in Pixels=Constant() {r=1}) {materialFunction=&MF_UI_CoreUI_RoundedMask}
  n1 = MaterialFunctionCall(UVs=NamedRerouteUsage(), Texture Scale=NamedRerouteUsage()) {materialFunction=&ScaleUVsByCenter}
  n2 = Constant2Vector() {g=1.8, r=1.2}
  n3 = Saturate(None=Add(
    A=MaterialFunctionCall(UVs=@n1.UVs, Texture Scale=@n2) {materialFunction=&ScaleUVsByCenter}.UVs
    B=Constant2Vector() {g=0.102, r=-0.166}
   ) {constB=0.5})
  n4 = TextureObject() {samplerType=SAMPLERTYPE_Alpha, texture=/DiscoveryBrowser/Art/Textures/T_UI_Refrigerator_Background_flare}
  NamedRerouteDeclaration(None=Multiply(A=TextureSample(UVs=@n3, Tex=@n4) {samplerType=SAMPLERTYPE_Alpha}.R) {constB=2}) {name=flare mask, nodeColor=<rgba 0 0 0 1>}
  n5 = MaterialFunctionCall() {materialFunction=/FortUILibrary/Systems/ModularButton/MaterialFunctions/MF_UIKit_ButtonStatesBase_Params}
  NamedRerouteDeclaration(None=@n5.Trans_Focused) {name=Is Focused, nodeColor=<rgba 0 0 0 1>}
  n6 = NamedRerouteUsage()
  n7 = AppendVector(A=Constant() {r=0.15}, B=Constant() {r=0.15})
  n8 = NamedRerouteUsage()
  n9 = NamedRerouteUsage()
  n10 = MaterialFunctionCall(
   UVs=MaterialFunctionCall!(
    ScreenUV=MaterialFunctionCall!(Input UVs=@n6, Translate=@n7) {materialFunction=/FortUILibrary/MaterialFunctions/Basics/MF_UI_Translate}.UVs raw
    TextureSize
    ScalePivotPoint
   ) {materialFunction=/FortUILibrary/MaterialFunctions/Basics/MF_UI_ScreenSpaceUVScaling}.Scaled to Fit
   Size=LinearInterpolate(A=Constant(), B=Constant() {r=4.5}, Alpha=LinearInterpolate(B=@n8, Alpha=@n9))
   OutlineThickness=Constant() {r=0.12}
  ) {materialFunction=&MF_UI_Circle}
  n11 = Constant() {r=0.04}
  n12 = OneMinus(None=MaterialFunctionCall!(In=Subtract(A=@n10.SDFOutline, B=@n11) {constB=0.005}, Ramp Width=Constant() {r=0.5}) {materialFunction=/Engine/Functions/Engine_MaterialFunctions01/Math/SmoothCeil}.Result)
  n13 = NamedRerouteUsage()
  n14 = Multiply(A=TextureSample(
    UVs=Panner(Coordinate=@n13) {speedX=-0.04, speedY=0.226}
    Tex=TextureObject() {texture=/DiscoveryBrowser/Art/Textures/T_UI_HonMoon_Repeatable_BlurMask}
   ) {samplerType=SAMPLERTYPE_Alpha}.R) {constB=2}
  n15 = Multiply(A=Power(Base=@n14), B=@n14)
  n16 = Multiply(A=@n14) {constB=0.9}
  NamedRerouteDeclaration(None=Multiply(A=Saturate(None=@n12), B=Add(A=@n15, B=@n16))) {name=Inner Ring Mask, nodeColor=<rgba 0 0 0 1>}
  n17 = Subtract(A=MaterialFunctionCall(UV=MaterialFunctionCall() {materialFunction=&GetUserInterfaceUV}.Normalized UV) {materialFunction=&MF_UI_UVScaling}.Scaled to Fit) {constB=0.5}
  NamedRerouteDeclaration(None=Saturate(None=Length(None=@n17) {is=class MaterialExpressionLength})) {name=Background Mask, nodeColor=<rgba 0 0 0 1>}
  n18 = NamedRerouteUsage()
  n19 = NamedRerouteUsage()
  n20 = Constant() {r=-1.9}
  n21 = Constant() {r=-0.3}
  n22 = MaterialFunctionCall(
   UVs=MaterialFunctionCall(
    Input UVs=MaterialFunctionCall(UVs=@n18, Texture Scale=@n19) {materialFunction=&ScaleUVsByCenter}.UVs
    Translate=AppendVector(A=@n20, B=@n21)
   ) {materialFunction=&MF_UI_Translate}.UVs raw
   Texture Scale=AppendVector(A=Constant() {r=1.5}, B=Constant() {r=5})
  ) {materialFunction=&ScaleUVsByCenter}
  n23 = Multiply(A=@n22.UVs, B=NamedRerouteUsage()) {constB=2}
  n24 = Constant() {r=0.002}
  n25 = Constant() {r=2}
  n26 = Constant()
  n27 = MaterialFunctionCall(
   UVs=MaterialFunctionCall!(
    UVs=Add(A=@n22.UVs, B=@n23)
    Amplitude X=@n24
    Amplitude Y=@n24
    Frequency X=@n25
    Frequency Y=@n25
    Phase=Constant()
    Motion=MaterialFunctionCall!(FloatValue, Speed=@n26) {materialFunction=/FortUILibrary/MaterialFunctions/Basics/MF_UI_LinearTime}.0-1 Loop
   ) {materialFunction=/FortUILibrary/MaterialFunctions/Effects/MF_UI_WaveWarp}.Warped UVs
   "Rotation Angle (0-1)"=Constant() {r=0.3}
  ) {materialFunction=&CustomRotator}
  n28 = Constant2Vector() {r=-0.075}
  n29 = Multiply(A=Clamp(None=Add(A=@n27.Rotated Values, B=@n28)) {minDefault=0.3}, B=Constant2Vector() {g=1, r=24})
  n30 = NamedRerouteDeclaration(None=MaterialFunctionCall!(
    UVs=Frac(None=@n29)
    PointA=Constant2Vector() {g=-2}
    PointB=Constant2Vector() {g=2}
    Thickness
    OutlineThickness
    OutlineAlignment
    GlowMin
    GlowMax
   ) {materialFunction=/FortUILibrary/MaterialFunctions/Shapes/MF_UI_LineSegment}.Fill) {name=LinesMask_2, nodeColor=<rgba 0 0 0 1>}
  n31 = Constant() {r=0.3}
  n32 = NamedRerouteUsage()
  n33 = Constant2Vector() {g=0.082, r=-1.05}
  n34 = MaterialFunctionCall(UV=MaterialFunctionCall(
    UVs=MaterialFunctionCall(Input UVs=@n32, Translate=@n33) {materialFunction=&MF_UI_Translate}.UVs raw
    Texture Scale=Constant2Vector() {g=1, r=1}
   ) {materialFunction=&ScaleUVsByCenter}.UVs) {materialFunction=&MF_UI_RadialGradient}
  NamedRerouteDeclaration(None=LinearInterpolate(B=Multiply(A=@n30, B=@n31), Alpha=OneMinus(None=@n34.Smooth))) {name=Honmoon Mask_2, nodeColor=<rgba 0 0 0 1>}
  n35 = MaterialFunctionCall(UV=NamedRerouteDeclaration(None=MaterialFunctionCall() {materialFunction=&GetUserInterfaceUV}.Normalized UV) {name=UV, nodeColor=<rgba 0 0 0 1>}) {materialFunction=&MF_UI_UVScaling}
  NamedRerouteDeclaration(None=@n35.Scaled to Fit) {name=Scale to fit, nodeColor=<rgba 0 0 0 1>}
  NamedRerouteDeclaration(None=@n35.ScaleY) {name=Scale Y, nodeColor=<rgba 0 0 0 1>}
  n36 = NamedRerouteUsage()
  n37 = NamedRerouteUsage()
  n38 = Constant() {r=0.26}
  n39 = Constant() {r=-1.5}
  n40 = MaterialFunctionCall(
   UVs=MaterialFunctionCall(
    Input UVs=MaterialFunctionCall(UVs=@n36, Texture Scale=@n37) {materialFunction=&ScaleUVsByCenter}.UVs
    Translate=AppendVector(A=@n38, B=@n39)
   ) {materialFunction=&MF_UI_Translate}.UVs raw
   Texture Scale=AppendVector(A=Constant() {r=1.5}, B=Constant() {r=5})
  ) {materialFunction=&ScaleUVsByCenter}
  n41 = Add(A=@n40.UVs, B=Multiply(A=@n40.UVs, B=NamedRerouteUsage()) {constB=2})
  n42 = Constant() {r=0.002}
  n43 = Constant() {r=0.01}
  n44 = Constant() {r=2}
  n45 = ScalarParameter() {group=Distortion, parameterName=Ditortion_MotionPhase}
  n46 = MaterialFunctionCall(Speed=Constant()) {materialFunction=&MF_UI_LinearTime}
  n47 = Multiply(
   A=MaterialFunctionCall(
    UVs=MaterialFunctionCall(UVs=@n41, Amplitude X=@n42, Amplitude Y=@n43, Frequency X=@n44, Frequency Y=@n44, Phase=@n45, Motion=@n46.0-1 Loop) {materialFunction=&MF_UI_WaveWarp}.Warped UVs
    "Rotation Angle (0-1)"=ScalarParameter() {defaultValue=0.25, parameterName=LineRotation}
   ) {materialFunction=&CustomRotator}.Rotated Values
   B=Constant2Vector() {g=1, r=80}
  )
  n48 = NamedRerouteDeclaration(None=MaterialFunctionCall(UVs=Frac(None=@n47), PointA=Constant2Vector() {g=-2}, PointB=Constant2Vector() {g=2}) {materialFunction=&MF_UI_LineSegment}.Fill) {name=LinesMask, nodeColor=<rgba 0 0 0 1>}
  n49 = Constant() {r=0.08}
  n50 = Subtract(A=ComponentMask(None=NamedRerouteUsage()) {g=true}) {constB=0.4}
  NamedRerouteDeclaration(None=LinearInterpolate(B=Multiply(A=@n48, B=@n49), Alpha=Saturate(None=@n50))) {name=Honmoon Mask, nodeColor=<rgba 0 0 0 1>}
  NamedRerouteDeclaration(None=Multiply(A=ScalarParameter() {parameterName=Is Hovered Amount}, B=ScalarParameter() {defaultValue=1, parameterName=Hovered Speed})) {name=Is Hovered Amount, nodeColor=<rgba 0 0 0 1>}
  n51 = LinearInterpolate(
   A=LinearInterpolate(B=Constant3Vector() {constant=<rgba 0.3036 0.1111 1 1>}, Alpha=NamedRerouteUsage())
   B=Constant3Vector() {constant=<rgba 0.305 0.1119 1 1>}
   Alpha=NamedRerouteUsage()
  )
  n52 = Constant3Vector() {constant=<rgba 0 0.2448 0.1866 1>}
  n53 = NamedRerouteUsage()
  NamedRerouteDeclaration(None=LinearInterpolate(A=LinearInterpolate(A=@n51, B=@n52, Alpha=@n53), B=Constant3Vector() {constant=<rgba 0 0.6615 0.4971 1>}, Alpha=NamedRerouteUsage())) {name=glow colours, nodeColor=<rgba 0 0 0 1>}
  n54 = LinearInterpolate(A=NamedRerouteUsage(), B=NamedRerouteUsage(), Alpha=NamedRerouteUsage())
  n55 = MaterialFunctionCall!(Blend=Constant3Vector() {constant=<rgba 0.7083 0.7083 0.7083 0>}, Base=@n54) {materialFunction=/Engine/Functions/Engine_MaterialFunctions03/Blends/Blend_ColorDodge}
  n56 = NamedRerouteUsage()
  n57 = LinearInterpolate(
   A=LinearInterpolate(A=LinearInterpolate(A=@n54, B=@n55.Result, Alpha=@n56), B=NamedRerouteUsage(), Alpha=NamedRerouteUsage())
   B=NamedRerouteDeclaration(None=Constant3Vector() {constant=<rgba 0.0635 0.5298 1 1>}) {name=Glow colour, nodeColor=<rgba 0 0 0 1>}
   Alpha=Multiply(A=NamedRerouteUsage()) {constB=4}
  )
  n58 = NamedRerouteUsage()
  n59 = NamedRerouteUsage()
  NamedRerouteDeclaration(None=LinearInterpolate(A=LinearInterpolate(A=@n57, B=@n58, Alpha=@n59), B=NamedRerouteUsage(), Alpha=NamedRerouteUsage())) {name=Background, nodeColor=<rgba 0 0 0 1>}
  n60 = Constant3Vector() {constant=<rgba 0.0103 0.0152 0.1302 1>}
  n61 = Constant3Vector() {constant=<rgba 0.007 0.0103 0.0887 1>}
  n62 = ComponentMask(None=NamedRerouteUsage()) {r=true}
  NamedRerouteDeclaration(None=LinearInterpolate(A=LinearInterpolate(A=@n60, B=@n61, Alpha=@n62), B=Constant3Vector() {constant=<rgba 0.0051 0.0067 0.0573 1>}, Alpha=NamedRerouteUsage())) {name=Background colour, nodeColor=<rgba 0 0 0 1>}
  n63 = MaterialFunctionCall(
   Input UVs=MaterialFunctionCall(UVs=NamedRerouteUsage(), Texture Scale=Constant() {r=1}) {materialFunction=&ScaleUVsByCenter}.UVs
   Translate=AppendVector(A=Constant() {r=-0.2}, B=Constant() {r=3.056})
  ) {materialFunction=&MF_UI_Translate}
  n64 = TextureObject() {texture=&T_UI_HonMoon_Repeatable_BlurMask}
  n65 = Constant() {r=1}
  n66 = NamedRerouteDeclaration(None=SmoothStep(Value=TextureSample!(UVs=@n63.UVs raw, Tex=@n64, Level=@n65, Apply View MipBias) {mipValueMode=TMVM_MipLevel}.R)) {name=DistortionTexture_2, nodeColor=<rgba 0 0 0 1>}
  NamedRerouteDeclaration(None=Multiply(A=@n66, B=@n66)) {name=DistortionMap_2, nodeColor=<rgba 0 0 0 1>}
  n67 = Constant3Vector() {constant=<rgba 0 0.4237 1 1>}
  n68 = Constant3Vector() {constant=<rgba 0.1198 0.4927 1 1>}
  n69 = MaterialFunctionCall() {materialFunction=&MF_UI_RadialGradient}
  NamedRerouteDeclaration(None=LinearInterpolate(
    A=Constant3Vector() {constant=<rgba 0.4508 0.011 1 1>}
    B=LinearInterpolate(A=@n67, B=@n68, Alpha=@n69.Smooth)
    Alpha=MaterialFunctionCall() {materialFunction=&LinearGradient}.UGradient
   )) {name=HonmoonColour X, nodeColor=<rgba 0 0 0 1>}
  n70 = NamedRerouteUsage()
  n71 = NamedRerouteUsage()
  NamedRerouteDeclaration(None=LinearInterpolate(B=Constant() {r=1.125}, Alpha=LinearInterpolate(A=@n70, Alpha=@n71)) {constA=1}) {name=Hover Animation, nodeColor=<rgba 0 0 0 1>}
  n72 = Constant3Vector() {constant=<rgba 0.2188 0.5328 1 1>}
  n73 = Constant3Vector() {constant=<rgba 0.3385 0.6894 1 1>}
  n74 = MaterialFunctionCall() {materialFunction=&MF_UI_RadialGradient}
  n75 = MaterialFunctionCall() {materialFunction=&LinearGradient}
  NamedRerouteDeclaration(None=LinearInterpolate(
    A=Constant3Vector() {constant=<rgba 0.4508 0.011 1 1>}
    B=LinearInterpolate(A=@n72, B=@n73, Alpha=@n74.Smooth)
    Alpha=OneMinus(None=@n75.VGradient)
   )) {name=HonmoonColour Y, nodeColor=<rgba 0 0 0 1>}
  NamedRerouteDeclaration(None=@n5.Trans_Hovered) {name=Is Hovered, nodeColor=<rgba 0 0 0 1>}
  n76 = NamedRerouteUsage()
  n77 = NamedRerouteUsage()
  NamedRerouteDeclaration(None=Multiply(A=LinearInterpolate(B=@n76, Alpha=@n77)) {constB=50}) {name=Honmoon Pulse Mask, nodeColor=<rgba 0 0 0 1>}
  n78 = Multiply(A=Time()) {constB=0.5}
  NamedRerouteDeclaration(None=Frac(None=Add(A=@n78) {constB=0.5})) {name=Animation, nodeColor=<rgba 0 0 0 1>}
  n79 = NamedRerouteUsage()
  NamedRerouteDeclaration(None=Multiply(A=Multiply(A=@n79) {constB=10}, B=NamedRerouteUsage())) {name=Honmoon Glow, nodeColor=<rgba 0 0 0 1>}
  n80 = MaterialFunctionCall(
   Input UVs=MaterialFunctionCall(UVs=NamedRerouteUsage(), Texture Scale=Constant() {r=1.8}) {materialFunction=&ScaleUVsByCenter}.UVs
   Translate=AppendVector(A=Constant() {r=0.3}, B=Constant())
  ) {materialFunction=&MF_UI_Translate}
  n81 = TextureObject() {texture=&T_UI_HonMoon_Repeatable_BlurMask}
  n82 = Constant() {r=1}
  n83 = NamedRerouteDeclaration(None=SmoothStep(Value=TextureSample(UVs=@n80.UVs raw, Tex=@n81, Level=@n82) {mipValueMode=TMVM_MipLevel}.R)) {name=DistortionTexture, nodeColor=<rgba 0 0 0 1>}
  NamedRerouteDeclaration(None=Multiply(A=Multiply(A=@n83, B=@n83), B=ScalarParameter() {defaultValue=1.2, group=Distortion, parameterName=Distortion_Strength})) {name=DistortionMap, nodeColor=<rgba 0 0 0 1>}
  n84 = MaterialFunctionCall(UVs=Add(A=NamedRerouteUsage(), B=Constant2Vector() {g=0.034, r=0.55}), Texture Scale=Constant2Vector() {g=2.5874, r=1.09}) {materialFunction=&ScaleUVsByCenter}
  n85 = Multiply(A=MaterialFunctionCall("Value (0-1)"=MaterialFunctionCall(UV=@n84.UVs) {materialFunction=&MF_UI_RadialGradient}.Smooth) {materialFunction=&MF_UI_EaseCurves}.EaseInOutSine) {constB=0.5}
  n86 = NamedRerouteUsage()
  n87 = Constant2Vector() {g=-0.7, r=-0.13}
  n88 = MaterialFunctionCall(UV=MaterialFunctionCall(UVs=Add(A=@n86, B=@n87), Texture Scale=Constant2Vector() {g=1.3, r=1}) {materialFunction=&ScaleUVsByCenter}.UVs) {materialFunction=&MF_UI_RadialGradient}
  n89 = NamedRerouteDeclaration(None=Multiply(A=MaterialFunctionCall("Value (0-1)"=@n88.Smooth) {materialFunction=&MF_UI_EaseCurves}.EaseInOutSine) {constB=0.6}) {name=glow 2, nodeColor=<rgba 0 0 0 1>}
  n90 = NamedRerouteUsage()
  n91 = Constant2Vector() {g=1.3, r=-0.12}
  n92 = MaterialFunctionCall(UV=MaterialFunctionCall(UVs=Add(A=@n90, B=@n91), Texture Scale=Constant2Vector() {g=6, r=-1}) {materialFunction=&ScaleUVsByCenter}.UVs) {materialFunction=&MF_UI_RadialGradient}
  n93 = NamedRerouteDeclaration(None=Multiply(A=MaterialFunctionCall("Value (0-1)"=@n92.Smooth) {materialFunction=&MF_UI_EaseCurves}.EaseInOutSine) {constB=0.5}) {name=glow 1, nodeColor=<rgba 0 0 0 1>}
  NamedRerouteDeclaration(None=LinearInterpolate(A=NamedRerouteDeclaration(None=@n85) {name=glow3, nodeColor=<rgba 0 0 0 1>}, Alpha=LinearInterpolate(A=@n89, Alpha=@n93))) {name=Glow Masks, nodeColor=<rgba 0 0 0 1>}
  NamedRerouteUsage()
 ]
}
