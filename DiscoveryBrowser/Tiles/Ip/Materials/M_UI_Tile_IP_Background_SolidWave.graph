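# Translucent UI material for the SolidWave background of IP tiles in the Discovery browser.
# A slightly rotated UV grid uses random per-cell values to pick the R or G channel of wave mask and base colour textures.
# Button-state transitions and hover parameters are exposed as reroutes, with interpolators driving pattern scale and rotation.
# Blue gradients, a powered radial highlight and a rounded-corner mask function supply shading and shape.
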Material /DiscoveryBrowser/Tiles/Ip/Materials/M_UI_Tile_IP_Background_SolidWave
graph {
 properties {
  bAutomaticallySetUsageInEditor=false
  bUsedWithStaticMesh=false
  blendMode=BLEND_Translucent
  materialDomain=MD_UI
  usedShadingModels=Default Lit
 }
 nodes [
  Constant3Vector() {constant=<rgba 1 0.0999 0.3968 1>}
  n1 = NamedRerouteUsage()
  LinearInterpolate(Alpha=@n1) {constA=-0.18, constB=-0.2}
  MaterialFunctionCall(Corner Radius=Constant() {r=10}, Radius in Pixels=Constant() {r=1}) {materialFunction=&MF_UI_CoreUI_RoundedMask}
  n2 = MaterialFunctionCall() {materialFunction=/FortUILibrary/Systems/ModularButton/MaterialFunctions/MF_UIKit_ButtonStatesBase_Params}
  NamedRerouteDeclaration(None=@n2.Trans_Hovered) {name=Is Hovered, nodeColor=<rgba 0 0 0 1>}
  NamedRerouteDeclaration(None=@n2.Trans_Focused) {name=Is Focused, nodeColor=<rgba 0 0 0 1>}
  n3 = NamedRerouteUsage()
  n4 = Rotator(
   Coordinate=MaterialFunctionCall(UVs=NamedRerouteUsage(), Texture Scale=LinearInterpolate(Alpha=@n3) {constA=0.5, constB=0.9}) {materialFunction=&ScaleUVsByCenter}.UVs
   Time=LinearInterpolate(Alpha=@n1) {constA=0.18, constB=0.01}
  ) {speed=6.2832}
  n5 = TextureSample(UVs=@n4) {samplerType=SAMPLERTYPE_LinearColor, texture=/DiscoveryBrowser/Art/Textures/T_UI_SolidWave_Background_Mask}
  n6 = NamedRerouteUsage()
  n7 = NamedRerouteUsage()
  n8 = MaterialFunctionCall(
   UVs=MaterialFunctionCall() {materialFunction=&GetUserInterfaceUV}.Normalized UV
   Texture Scale=LinearInterpolate(B=Constant() {r=1.125}, Alpha=LinearInterpolate(A=@n6, Alpha=@n7)) {constA=1}
  ) {materialFunction=&ScaleUVsByCenter}
  n9 = MaterialFunctionCall(
   UVs=Multiply(A=MaterialFunctionCall(UV=@n8.UVs) {materialFunction=&MF_UI_UVScaling}.Scaled to Fit) {constB=2}
   "Rotation Angle (0-1)"=Constant() {r=-0.1}
  ) {materialFunction=&CustomRotator}
  n10 = MaterialFunctionCall!(UV=Floor(None=@n9.Rotated Values), Seed=Constant() {r=10}) {materialFunction=&MF_UI_Random}
  n11 = Saturate(None=Multiply(A=NamedRerouteDeclaration(None=@n10.Result) {name=Random Cells, nodeColor=<rgba 0 1 0.4375 1>}, B=Constant() {r=1.5}) {constB=1.3})
  n12 = Floor(None=@n11)
  n13 = Min(A=OneMinus(None=@n12), B=Round(None=@n11))
  NamedRerouteDeclaration(None=LinearInterpolate(A=LinearInterpolate(A=@n5.G, B=@n5.G, Alpha=@n13), B=@n5.R, Alpha=@n12)) {name=Pattern Alpha, nodeColor=<rgba 0 0 0 1>}
  n14 = TextureSample(UVs=@n4) {samplerType=SAMPLERTYPE_LinearColor, texture=/DiscoveryBrowser/Art/Textures/T_UI_SolidWave_Background_BaseColour}
  NamedRerouteDeclaration(None=LinearInterpolate(A=LinearInterpolate(A=@n14.G, B=@n14.G, Alpha=@n13), B=@n14.R, Alpha=@n12)) {name=Pattern Colour, nodeColor=<rgba 0 0 0 1>}
  NamedRerouteDeclaration(None=Frac(None=@n9.Rotated Values)) {name=Pattern UV, nodeColor=<rgba 0.5234 1 0 1>}
  NamedRerouteDeclaration(None=Multiply(A=ScalarParameter() {parameterName=Is Hovered Amount}, B=ScalarParameter() {defaultValue=1, parameterName=Hovered Speed})) {name=Is Hovered Amount, nodeColor=<rgba 0 0 0 1>}
  n15 = MaterialFunctionCall() {materialFunction=&GetUserInterfaceUV}
  NamedRerouteDeclaration(None=LinearInterpolate(
    A=Constant3Vector() {constant=<rgba 0.1022 0.4397 0.5776 1>}
    B=Constant3Vector() {constant=<rgba 0.0356 0.2462 0.402 1>}
    Alpha=ComponentMask(None=@n15.Normalized UV) {g=true}
   )) {name=Gradient, nodeColor=<rgba 0 0 0 1>}
  n16 = MaterialFunctionCall() {materialFunction=&GetUserInterfaceUV}
  n17 = Constant() {r=-0.1}
  n18 = ComponentMask(None=Subtract(A=MaterialFunctionCall(UVs=@n16.Normalized UV, "Rotation Angle (0-1)"=@n17) {materialFunction=&CustomRotator}.Rotated Values) {constB=0.6}) {g=true}
  NamedRerouteDeclaration(None=Saturate(None=Multiply(A=@n18) {constB=0.7})) {name=Gradient Mask, nodeColor=<rgba 0 0 0 1>}
  NamedRerouteDeclaration(None=NamedRerouteUsage()) {name=Corner mask, nodeColor=<rgba 0 0 0 1>}
  n19 = Constant3Vector() {constant=<rgba 0.0343 0.2423 0.3968 1>}
  n20 = Constant3Vector() {constant=<rgba 0.1022 0.4397 0.5776 1>}
  n21 = Multiply(A=@n20, B=NamedRerouteUsage())
  n22 = LinearInterpolate(A=LinearInterpolate(A=NamedRerouteUsage(), B=@n20, Alpha=@n21), B=@n21, Alpha=Multiply(A=NamedRerouteUsage()) {constB=0.4})
  n23 = Constant3Vector() {constant=<rgba 0.5647 0.6105 0.5711 1>}
  n24 = Add(
   A=MaterialFunctionCall() {materialFunction=&GetUserInterfaceUV}.Normalized UV
   B=AppendVector(A=Constant() {r=-0.3}, B=Constant() {r=0.1})
  )
  n25 = Constant() {r=5.2}
  n26 = Power(
   Base=MaterialFunctionCall(UV=MaterialFunctionCall(UVs=@n24, Texture Scale=@n25) {materialFunction=&ScaleUVsByCenter}.UVs) {materialFunction=&MF_UI_RadialGradient}.Smooth
   Exp=Constant() {r=7}
  )
  n27 = LinearInterpolate(A=@n22, B=Saturate(None=Add(A=@n22, B=@n23)), Alpha=Multiply(A=@n26) {constB=0.2})
  Saturate(None=LinearInterpolate(A=LinearInterpolate(A=@n19, B=@n27, Alpha=@n26), B=Constant3Vector() {constant=<rgba 0 0 0 1>}, Alpha=NamedRerouteUsage()) {constB=0})
 ]
}
