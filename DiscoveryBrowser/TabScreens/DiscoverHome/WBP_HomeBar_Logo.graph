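# WBP_HomeBar_Logo is a single-image widget that displays a logo texture for the home bar in the Discovery browser's DiscoverHome screen.
# SetTexture ignores zero sizes, resizes the brush, and passes the texture to the image's dynamic material.
# SetTextureDimensionFormat uses GetAspectRatio to give square logos a fixed size and fit others within width or height caps, preserving proportions.
# PreConstruct previews a valid Texture variable at design time with a placeholder square size.
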WidgetBlueprint /DiscoveryBrowser/TabScreens/DiscoverHome/WBP_HomeBar_Logo
graphs [
 EventGraph {nodes [
   n1 = Event PreConstruct()
   n2 = Get Texture() -> (Texture:Texture 2D Object Reference)
   > Branch(Condition=AND Boolean(A=@n1.IsDesignTime, B=IsValid(Object=@n2)))
   > SetTexture!(execute, self, TextureSize:Vector 2D Structure=(X=256,Y=256), Texture:Texture Object Reference=@n2) -> (then)
  ]}
 GetAspectRatio {nodes [
   n1 = GetAspectRatio!() -> (then, Size_X:double, Size_Y:double) {is=title K2Node_FunctionEntry}
   > Branch(Condition=AND Boolean(A="float > float"(A=@n1), B="float > float"(A=@n1.Size_Y)))
   n2 = MakeVector2D!(X=float / float(A=@n1, B=@n1.Size_Y), Y=1) -> (ReturnValue_X:double, ReturnValue_Y:double)
   n3 = MakeVector2D(X=1, Y=float / float(A=@n1.Size_Y, B=@n1))
   > Return Node!(
    execute
    AspectRatio:Vector 2D Structure=MakeVector2D!(X="Max (Float)"(A=@n2, B=@n3), Y="Max (Float)"(A=@n2.ReturnValue_Y, B=@n3.ReturnValue_Y)) -> (ReturnValue:Vector 2D Structure)
   )
  ]}
 SetTexture {nodes [
   n1 = SetTexture!() -> (then, TextureSize:Vector 2D Structure, Texture:Texture Object Reference) {is=title K2Node_FunctionEntry}
   > n2 = Branch(Condition=IsZero2D!(A:"Vector 2D Structure (by ref)"=@n1) -> (ReturnValue:bool))
   n3 = SetTextureDimensionFormat!(execute=@n2.else, self, TextureSize:Vector 2D Structure=@n1) -> (then, NewSize:Vector 2D Structure)
   n4 = Get!(self:Image Object Reference=Get Logo() -> (Logo:Image Object Reference)) -> (
    Brush_TintColor:Slate Color Structure
    Brush_DrawAs:ESlateBrushDrawType Enum
    Brush_Tiling:ESlateBrushTileType Enum
    Brush_Mirroring:ESlateBrushMirrorType Enum
    Brush_ImageSize:Vector 2D Structure
    Brush_Margin:Margin Structure
    Brush_ResourceObject:Object Reference
    Brush_OutlineSettings:Slate Brush Outline Settings Structure
   )
   > SetBrush!(
    execute
    self=Get Logo()
    InBrush_TintColor:Slate Color Structure=@n4
    InBrush_DrawAs:ESlateBrushDrawType Enum=@n4.Brush_DrawAs
    InBrush_Tiling:ESlateBrushTileType Enum=@n4.Brush_Tiling
    InBrush_Mirroring:ESlateBrushMirrorType Enum=@n4.Brush_Mirroring
    InBrush_ImageSize:Vector 2D Structure=@n3
    InBrush_Margin:Margin Structure=@n4.Brush_Margin
    InBrush_ResourceObject:Object Reference=@n4.Brush_ResourceObject
    InBrush_OutlineSettings:Slate Brush Outline Settings Structure=@n4.Brush_OutlineSettings
   ) -> (then)
   > n5 = GetDynamicMaterial(self=Get Logo())
   > SetTextureParameterValue(self=@n5, ParameterName=Texture, Value=@n1.Texture)
  ]}
 SetTextureDimensionFormat {nodes [
   n1 = SetTextureDimensionFormat!() -> (then, TextureSize:Vector 2D Structure) {is=title K2Node_FunctionEntry}
   > n2 = GetAspectRatio!(execute, self, Size:Vector 2D Structure=@n1) -> (then, AspectRatio_X:double, AspectRatio_Y:double) {is=title K2Node_CallFunction}
   n3 = "To Float (Integer)"(InInt=Get SquareSize() -> (SquareSize:int=0))
   n4 = Get!() -> (TextureSize_X:double, TextureSize_Y:double)
   n5 = float * float(A:int=Get Cap Height() -> (Cap Height:int=0), B=@n4) -> (ReturnValue:int) {is=title K2Node_PromotableOperator}
   n6 = MakeVector2D(
    X="To Float (Integer)"(InInt=float / float(A:int=@n5, B=@n4.TextureSize_Y) -> (ReturnValue:int))
    Y="To Float (Integer)"(InInt=Get Cap Height() -> (Cap Height=0))
   )
   n7 = "float > float"(A=@n2, B=@n2.AspectRatio_Y)
   n8 = Get Cap Height() -> (Cap Height=0)
   n9 = Get()
   n10 = float * float(A:double=@n9.TextureSize_Y, B:int=Get Cap Width() -> (Cap Width:int=0))
   n11 = MakeVector2D!(X="To Float (Integer)"(InInt=Get Cap Width() -> (Cap Width=0)), Y="To Float (Integer)"(InInt=float / float(A=@n10, B=@n9))) -> (ReturnValue_X:double, ReturnValue_Y:double)
   n12 = float / float(A=float * float(A=@n11, B=Get Cap Height() -> (Cap Height=0)), B=@n11.ReturnValue_Y)
   n13 = Select!(
    Option 0_X:double=@n11
    Option 0_Y:double=@n11.ReturnValue_Y
    Option 1:Vector 2D Structure=MakeVector2D!(X="To Float (Integer)"(InInt=@n12), Y="To Float (Integer)"(InInt=@n8)) -> (ReturnValue:Vector 2D Structure)
    Index="float > float"(A=@n11.ReturnValue_Y, B:int=Get Cap Height() -> (Cap Height=0))
   ) -> (ReturnValue:Vector 2D Structure)
   > Return Node!(
    execute
    NewSize:Vector 2D Structure=Select!(
     Option 0:Vector 2D Structure=Select!(Option 0=@n6, Option 1:Vector 2D Structure=@n13, Index=@n7) -> (ReturnValue)
     Option 1_X:double=@n3
     Option 1_Y:double=@n3
     Index="Equal (Float)"(A=@n2, B=@n2.AspectRatio_Y)
    ) -> (ReturnValue)
   )
  ]}
]
variables [
 Cap Height {type=integer, default=320}
 Cap Width {type=integer, default=768}
 SquareSize {type=integer, default=192}
 Texture {type=Texture2D}
]
widgets {
 info {inheritedWidgetCount=0, parentClass=&UserWidget, rootWidgetClass=&Image, widgetCount=1}
 widgets [Logo {
   widgetClassPath=&Image
   bIsVariable=true
   widget=Logo
   properties {
    visibility=HitTestInvisible
    brush {
     bIsDynamicallyLoaded=false
     drawAs=Image
     imageType=NoImage
     mirroring=NoMirror
     resourceName=None
     resourceObject=/DiscoveryBrowser/TabScreens/DiscoverHome/Materials/MI_UI_HomeBar_Logo
     tiling=NoTile
     imageSize=<32 32>
     margin=<ltrb 0 0 0 0>
     outlineSettings {
      bUseBrushTransparency=false
      roundingType=HalfHeightRadius
      width=0
      color {colorUseRule=UseColor_Specified, specifiedColor=<rgba 0 0 0 0>}
      cornerRadii=<0 0 0 0>
     }
     tintColor {colorUseRule=UseColor_Specified, specifiedColor=<rgba 1 1 1 1>}
     uVRegion {bIsValid=false, max=<0 0>, min=<0 0>}
    }
   }
  }]
}
