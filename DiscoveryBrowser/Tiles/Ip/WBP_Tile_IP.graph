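# WBP_Tile_IP is a generic UIKit button for the Discovery browser IP tile, layering a logo over foreground art, with Fall Guys debug defaults.
# PreConstruct applies materials, textures and rounded corners; setters swap background material and foreground texture.
# Hover loops an animation on HoveredAdditive, whose setter drives the background material, while mouse entry passes local cursor position to the foreground material.
# Focus visuals show only for gamepad input, refreshing on input-method changes.
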WidgetBlueprint /DiscoveryBrowser/Tiles/Ip/WBP_Tile_IP
animations [IsHoveredIncrease {
  binding_count=1
  length_seconds=1
  object=&self:IsHoveredIncrease
  sequence {bindings [WBP_Tile_IP_C_144 {
     id=#1
     possessed_class=WBP_Tile_IP_C
     tracks [{
       type=MovieSceneDoubleTrack
       display_name=Hovered Additive
       sections [{type=MovieSceneDoubleSection, end_frame=20, start_frame=0, channels [None {type=MovieSceneScriptingDoubleChannel, keys [[0,0],[20,1]]}]}]
      }]
    }]}
 }]
graphs [
 EventGraph {nodes [
   Event On Focused() {is=title K2Node_Event}
   > n1 = Parent: On Focused() {is=title K2Node_CallParentFunction}
   RefreshFocusVisual(execute=[@n1.then, @n2.then, @n4.then])
   Event On Unfocused() {is=title K2Node_Event}
   > n2 = Parent: On Unfocused() {is=title K2Node_CallParentFunction}
   Event On Hovered()
   > Parent: On Hovered() {is=title K2Node_CallParentFunction}
   n3 = Get Foreground() -> (Foreground:WBP UIKit Block Image Object Reference)
   > RequestTransition!(execute, self:UIKit Block Object Reference=@n3, State:EUIKitBlockVisualState Enum=Hovered) -> (then)
   > Set HoveredAdditive(HoveredAdditive:double=0) -> (Output_Get:double)
   > PlayAnimation(InAnimation=Get IsHoveredIncrease() -> (IsHoveredIncrease:Widget Animation Object Reference), NumLoopsToPlay=0, bRestoreState=true)
   Event On Unhovered()
   > Parent: On Unhovered() {is=title K2Node_CallParentFunction}
   > RequestTransition(self=@n3, State=Unhovered)
   > StopAnimation(InAnimation=Get IsHoveredIncrease())
   n4 = Event On Input Method Changed!() -> (OutputDelegate, then, CurrentInputType:ECommonInputType Enum) {is=title K2Node_Event}
   n5 = Event PreConstruct()
   > SetMaterial!(
    execute
    self:WBP UIKit Block Image Object Reference=Get Foreground()
    InMaterial:"Material Instance Object Reference (by ref)"=Get ForegroundMaterialInstance() -> (ForegroundMaterialInstance:Material Instance Object Reference)
   ) -> (then)
   > SetTexture!(
    execute
    self:WBP Home Bar Logo Object Reference=Get Logo() -> (Logo:WBP Home Bar Logo Object Reference)
    TextureSize:Vector 2D Structure=(X=256,Y=256)
    Texture:Texture Object Reference=Get DebugLogoTexture() -> (DebugLogoTexture:Texture 2D Object Reference)
   ) -> (then)
   > SetForegroundTexture!(execute, self, Texture:Texture Object Reference=Get DebugForegroundTexture() -> (DebugForegroundTexture:Texture 2D Object Reference)) -> (then)
   > SetBackgroundMaterials!(
    execute
    self:WBP UIKit Button Generic Base Object Reference
    Background MI:Material Instance Object Reference=Get DebugBackgroundMaterialInstance() -> (DebugBackgroundMaterialInstance:Material Instance Object Reference)
    Outline MI:Material Instance Object Reference=Get OutlineMaterial() -> (OutlineMaterial:Material Instance Object Reference)
   ) -> (then)
   > Parent: PreConstruct(IsDesignTime=@n5.IsDesignTime)
   > n6 = Cast To WBP_UIKit_Block_BackgroundWithOutline!(execute, Object=Get Block_Background() -> (Block_Background:WBP UIKit Block Background Base Object Reference)) -> (then, CastFailed, AsWBP UIKit Block Background with Outline:WBP UIKit Block Background With Outline Object Reference) {is=title K2Node_DynamicCast}
   > SetCornerRadiuses!(
    execute
    self:WBP UIKit Block Background With Outline Object Reference=@n6
    TopLeft:float=18
    TopRight:float=18
    BottomLeft:float=18
    BottomRight:float=18
    RadiusInPixels:bool=true
   ) -> (then)
   n7 = Event OnMouseEnter()
   n8 = ScreenToLocal!(Geometry:"Geometry Structure (by ref)"=@n7.MyGeometry, ScreenPosition:Vector 2D Structure, bIncludeWindowPosition:bool=false) -> (LocalCoordinate_X:double, LocalCoordinate_Y:double)
   n9 = Get ImageMID!(self:WBP UIKit Block Image Object Reference=Get Foreground()) -> (ImageMID:Material Instance Dynamic Object Reference)
   > SetScalarParameterValue(self=@n9, ParameterName=ScreenPosX, Value=@n8)
   > SetScalarParameterValue(self=@n9, ParameterName=ScreenPosY, Value=@n8.LocalCoordinate_Y)
   > Parent: OnMouseEnter!(execute, MyGeometry:Geometry Structure=@n7.MyGeometry, MouseEvent:"Pointer Event Structure (by ref)"=@n7.MouseEvent) -> (then) {is=title K2Node_CallParentFunction}
  ]}
 RefreshFocusVisual {nodes [
   RefreshFocusVisual() {is=title K2Node_FunctionEntry}
   > n1 = Branch(Condition=HasAnyUserFocusedDescendants!(self:Widget Object Reference) -> (ReturnValue:bool))
   > n2 = Switch on ECommonInputType!(execute, Selection:ECommonInputType Enum=GetCurrentInputType(self=Get CommonInputSubsystem())) -> (MouseAndKeyboard:Exec, Gamepad:Exec, Touch:Exec, Count:Exec) {is=title K2Node_SwitchEnum}
   RequestTransition(execute=@n2.Gamepad, self=Get Foreground(), State=Focused)
   RequestTransition(execute=[@n2.MouseAndKeyboard, @n2.Touch, @n1.else], self=Get Foreground(), State=Unfocused)
  ]}
 SetBackgroundMaterialInstance {nodes [
   n1 = SetBackgroundMaterialInstance!() -> (then, BackgroundMaterialInstance:Material Instance Object Reference) {is=title K2Node_FunctionEntry}
   > SetBackgroundMaterials(Background MI=@n1, Outline MI=Get OutlineMaterial())
  ]}
 SetForegroundTexture {nodes [
   n1 = SetForegroundTexture!() -> (then, Texture:Texture Object Reference) {is=title K2Node_FunctionEntry}
   > SetTextureParameterValue(self=Get ImageMID(self=Get Foreground()), ParameterName=Texture, Value=@n1)
  ]}
 SetHoveredAdditive {nodes [
   n1 = SetHoveredAdditive!() -> (then, NewParam:double=~) {is=title K2Node_FunctionEntry}
   > n2 = Set HoveredAdditive(HoveredAdditive=@n1)
   > SetScalarParameterValue(
    self=Get BackgroundDynamicMaterial!(self:WBP UIKit Block Background Base Object Reference=Get Block_Background()) -> (BackgroundDynamicMaterial:Material Instance Dynamic Object Reference)
    ParameterName=Is Hovered Amount
    Value=float + float(A=@n2, B=@n1)
   )
  ]}
]
variables [
 DebugBackgroundMaterialInstance {type=MaterialInstance, default=/DiscoveryBrowser/Tiles/Ip/MaterialInstances/MI_UI_Tile_IP_Background_FallGuys}
 DebugForegroundTexture {type=Texture2D, default=/DiscoveryBrowser/Tiles/Ip/Textures/T_UI_Discovery_Tile_Foreground_FallGuys}
 DebugLogoTexture {type=Texture2D, default=/DiscoveryBrowser/Tiles/Ip/Textures/T_UI_Discovery_Tile_Logo_FallGuys}
 ForegroundMaterialInstance {type=MaterialInstance, default=/DiscoveryBrowser/Tiles/Ip/MaterialInstances/MI_UI_Tile_IP_Foreground}
 HoveredAdditive {type=number}
]
widgets {
 info {
  inheritedWidgetCount=0
  namedSlotCount=1
  parentClass=/FortUILibrary/UIKit/Buttons/WBP_UIKit_Button_Generic.WBP_UIKit_Button_Generic_C
  rootWidgetClass=null
  widgetCount=3
 }
 widgets [
  GP_Content {widgetClassPath=&GridPanel, widget=GP_Content, properties {columnFill [1], rowFill [1]}}
  Foreground {
   widgetClassPath=/FortUILibrary/Systems/ModularButton/Blocks/WBP_UIKit_Block_Image
   bIsVariable=true
   parent=GP_Content
   widget=Foreground
   properties {pixelSnapping=Disabled}
   slot {class=GridSlot, properties {padding=<ltrb 0 -12 0 0>}}
  }
  Logo {
   widgetClassPath=/DiscoveryBrowser/TabScreens/DiscoverHome/WBP_HomeBar_Logo
   bIsVariable=true
   parent=GP_Content
   widget=Logo
   properties {
    cap Height=80
    cap Width=144
    squareSize=88
    texture=&T_UI_Discovery_Tile_Logo_FallGuys
    visibility=HitTestInvisible
   }
   slot {class=GridSlot, properties {horizontalAlignment=HAlign_Left, layer=1, verticalAlignment=VAlign_Center, padding=<ltrb 24 0 0 0>}}
  }
 ]
}
