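# IP category tile button for the Discovery browser, showing a brand background, pop-out foreground and logo from its tile view model.
# View-bound functions pick category materials, fallback textures, favorite visibility and context-menu actions.
# Clicks push a nested collection surface, or toggle a pending favorite in the interests chooser row.
# A looping hover animation feeds the background material, while gamepad focus, optional parallax and mobile sizing adjust the visuals.
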WidgetBlueprint /DiscoveryBrowser/Tiles/Ip/WBP_Tile_IP_NEW
animations [IsHoveredIncrease {
  binding_count=1
  length_seconds=0.3472
  object=&self:IsHoveredIncrease
  sequence {bindings [WBP_Tile_IP_C_144 {
     id=#1
     possessed_class=WBP_Tile_IP_C
     tracks [{
       type=MovieSceneDoubleTrack
       display_name=Hovered Additive
       sections [{type=MovieSceneDoubleSection, end_frame=6, start_frame=0, channels [None {type=MovieSceneScriptingDoubleChannel, keys [[0,0],[6.943,1]]}]}]
      }]
    }]}
 }]
graphs [
 AddViewCollectionToContextMenu {nodes [
   n1 = AddViewCollectionToContextMenu!() -> (then, IsInInterestsRow:bool=~) {is=title K2Node_FunctionEntry}
   > n2 = Branch(Condition=@n1)
   SetupViewCollection(execute=@n2.else)
  ]}
 EventGraph {nodes [
   ToggleFavorite!() -> (OutputDelegate, then, InteractionNameID:Name) {is=title K2Node_CustomEvent}
   > n1 = Branch(Condition=Get bIsInInterestsChooserRow!(self:Discover Tile Item View Model Object Reference=Get FortDiscoverSurfaceTileItemVM() -> (FortDiscoverSurfaceTileItemVM:Fort Discover Surface Tile Item VM Object Reference)) -> (bIsInInterestsChooserRow:bool))
   n2 = Get FortDiscoverSurfaceTileItemVM()
   > TogglePendingFavorite(self:Fort Discover Surface Tile Item VM Object Reference=@n2)
   > n3 = Branch(Condition=Get bIsPendingFavorite!(self:Fort Discover Surface Tile Item VM Object Reference=@n2) -> (bIsPendingFavorite:bool))
   n4 = PlaySound2D(execute=[@n6.then, @n3.then], Sound=Get Favorited Sound() -> (Favorited Sound:Sound Base Object Reference))
   CloseContextMenu(
    execute=[@n5.then, @n4.then]
    self:WBP Discover Tile Base Object Reference=Get WBP_Discover_Tile_Base() -> (WBP_Discover_Tile_Base:WBP Discover Tile Base Object Reference)
   )
   n5 = PlaySound2D(execute=[@n6.else, @n3.else], Sound=Get Unfavorited Sound() -> (Unfavorited Sound:Sound Base Object Reference))
   ToggleFavorite(execute=@n1.else, self:Fort Discover Link Entry Tile Item VM Object Reference=@n2) {is=title K2Node_CallFunction}
   > n6 = Branch(Condition=Get bIsFavorite!(self:Fort Discover Link Entry Tile Item VM Object Reference=@n2) -> (bIsFavorite:bool))
   ViewCollection!() -> (OutputDelegate, then, InteractionID:Name) {is=title K2Node_CustomEvent}
   > n7 = CloseContextMenu(self=Get WBP_Discover_Tile_Base())
   n8 = Get FortDiscoverSurfaceTileItemVM()
   n9 = GetReferenceSurfaceName!(self:Fort Discover Surface Tile Item VM Object Reference=@n8) -> (ReturnValue:Name=None)
   SendFlowRiderClickTileEvent!(
    execute=[@n7.then, @n26.then]
    self:WBP Discover Tile Base Object Reference=Get WBP_Discover_Tile_Base()
    TileWidget:User Widget Object Reference=Self-Reference()
    TileType:EActivityItemType Enum=IPTile
    TileUniqueIdentifier:String="To String (Name)"(InName=@n9)
   ) -> (then)
   > HandleCategoryTileClicked!(
    execute
    self:Fort Activity Selector Object Reference=Get FortActivitySelector() -> (ReturnValue:Fort Activity Selector Object Reference) {is=title K2Node_GetSubsystem}
    FortPC:Fort Player Controller Object Reference=GetFirstLocalFortPlayerController!(bIncludeNonSpectators:bool=true, bIncludeSpectators:bool=false) -> (ReturnValue:Fort Player Controller Object Reference)
    InTileItemVM:Fort Discover Surface Tile Item VM Object Reference=Get FortDiscoverSurfaceTileItemVM()
   ) -> (then)
   > n10 = PushPrimaryContentClassAsync!(
    execute:Exec
    WidgetClass:Common Activatable Widget Class Reference=/DiscoveryBrowser/WBP_Discover_NestedSurface.WBP_Discover_NestedSurface_C
   ) -> (then:Exec, OnPrimaryContentPushed:Exec, ActivatableWidget:Common Activatable Widget Object Reference) {is=title K2Node_AsyncAction}
   n11 = Cast To WBP_Discover_NestedSurface!(execute=@n10.OnPrimaryContentPushed, Object=@n10) -> (then, CastFailed, AsWBP Discover Nested Surface:WBP Discover Nested Surface Object Reference) {is=title K2Node_DynamicCast}
   > SetFortDiscoverSurfaceTileItemVM!(execute, self:WBP Discover Nested Surface Object Reference=@n11, Viewmodel:Fort Discover Surface Tile Item VM Object Reference=@n8) -> (then)
   > On Disover Surface Tile Item VM Set(self:WBP Discover Nested Surface Object Reference=@n11)
   n12 = Get FortDiscoverSurfaceViewModel!(self:WBP Discover Nested Surface Object Reference=@n11) -> (FortDiscoverSurfaceViewModel:Discover Surface View Model Object Reference)
   > SetParentItem!(execute, self:Discover Surface View Model Object Reference=@n12, InParentItem:Fort Discover Surface Tile Item VM Object Reference=@n8) -> (then)
   > SetSurfaceName!(execute, self:Discover Surface View Model Object Reference=@n12, InSurfaceName:Name=@n9) -> (then)
   SetupViewCollection() {is=title K2Node_CustomEvent}
   n13 = Get WBP_Discover_Tile_Base()
   > RemoveInteractionByName!(
    execute
    self:WBP Discover Tile Base Object Reference=@n13
    ActionToRemove:Name=Get Action to Remove_Collection() -> (Action to Remove_Collection:Name=None)
   ) -> (then)
   > AddInteraction!(
    execute
    self:UIKit Context Menu Action Context Object Reference=Get ContextMenuActionContext!(self:WBP Discover Tile Base Object Reference=@n13) -> (ContextMenuActionContext:UIKit Context Menu Action Context Object Reference)
    InActionInfo_DisplayName:Text="LOCTEXT(\"View Collection\")"
    InActionInfo_IconTexture:Texture 2D Soft Object Reference
    InActionInfo_IconMaterial:Material Instance Soft Object Reference
    InActionInfo_SortPriority:int=1000
    InActionInfo_bIsAvailable:bool=True
    InActionInfo_bIsEnabled:bool=True
    InActionInfo_bIsSelected:bool=False
    InActionInfo_bAddDivider:bool=False
    InActionInfo_bPositionDividerAfter:bool=False
    InActionInfo_bCloseMenuOnSelect:bool=False
    InActionInfo_InteractionNameID:Name=Get Action to Remove_Collection() -> (Action to Remove_Collection=None)
    InActionInfo_Interaction:Data Table Row Handle Structure=(DataTable=None,RowName="")
    InActionInfo_PressedDelegate:Delegate="(null).None"
    InActionInfo_ReleasedDelegate:Delegate="(null).None"
    InInteractionDelegate:"Delegate (by ref)"=Create Event()
   ) -> (then)
   Event On Focused() {is=title K2Node_Event}
   > n14 = Parent: On Focused() {is=title K2Node_CallParentFunction}
   RefreshFocusVisual(execute=[@n14.then, @n15.then])
   Event On Unfocused() {is=title K2Node_Event}
   > n15 = Parent: On Unfocused() {is=title K2Node_CallParentFunction}
   n16 = Event PreConstruct()
   > Parent: PreConstruct(IsDesignTime=@n16.IsDesignTime)
   > SetMaterial!(
    execute
    self:WBP UIKit Block Image Object Reference=Get Foreground() -> (Foreground:WBP UIKit Block Image Object Reference)
    InMaterial:"Material Instance Object Reference (by ref)"=Get ForegroundMaterialInstance() -> (ForegroundMaterialInstance:Material Instance Object Reference)
   ) -> (then)
   > SetTexture!(
    execute
    self:WBP Home Bar Logo Object Reference=Get Logo() -> (Logo:WBP Home Bar Logo Object Reference)
    TextureSize:Vector 2D Structure=(X=256,Y=256)
    Texture:Texture Object Reference=Get DebugLogoTexture() -> (DebugLogoTexture:Texture 2D Object Reference)
   ) -> (then)
   > SetForegroundTexture!(execute, self, Texture:Texture Object Reference=Get DebugForegroundTexture() -> (DebugForegroundTexture:Texture 2D Object Reference)) -> (then)
   > SetBackgroundMaterial!(
    execute
    self:WBP UIKit Block Background Base Object Reference=Get Block_Background() -> (Block_Background:WBP UIKit Block Background Generic Object Reference)
    Material Instance:"Material Instance Object Reference (by ref)"=Get DebugBackgroundMaterialInstance() -> (DebugBackgroundMaterialInstance:Material Instance Object Reference)
   ) -> (then)
   > Set Tile Size!(execute, self:WBP Discover Tile Base Object Reference=Get WBP_Discover_Tile_Base(), InSize:Vector 2D Structure=(X=396,Y=144)) -> (then) {is=title K2Node_CallFunction}
   > SetCompactUIStyle() {is=title K2Node_CallFunction}
   Event On Hovered()
   > Parent: On Hovered() {is=title K2Node_CallParentFunction}
   n17 = Get Block_Background()
   n18 = Get Block_Outline!(self:WBP Discover Tile Base Object Reference=Get WBP_Discover_Tile_Base()) -> (Block_Outline:WBP UIKit Block Outline Object Reference)
   n19 = Get Foreground()
   > RequestTransition!(execute, self:UIKit Block Object Reference=[@n17, @n18, @n19], State:EUIKitBlockVisualState Enum=Hovered) -> (then)
   > Set HoveredAdditive(HoveredAdditive:double=0) -> (Output_Get:double)
   > PlayAnimation(InAnimation=Get IsHoveredIncrease() -> (IsHoveredIncrease:Widget Animation Object Reference), NumLoopsToPlay=0, bRestoreState=true)
   Event Construct()
   > Parent: Construct()
   n20 = GetFortGameUserSettings!() -> (ReturnValue:Fort Game User Settings Object Reference)
   > SetScalarParameterValue(
    self=Get ImageMID!(self:WBP UIKit Block Image Object Reference=Get Foreground()) -> (ImageMID:Material Instance Dynamic Object Reference)
    ParameterName=Parallax Offset
    Value=Select(
     Option 0:float=0
     Option 1:float=4
     Index=Get bAllowUIParallax!(self:Fort Game User Settings Object Reference=@n20) -> (bAllowUIParallax:bool)
    ) -> (ReturnValue:float)
   )
   Event On Unhovered()
   > Parent: On Unhovered() {is=title K2Node_CallParentFunction}
   > RequestTransition(self=[@n17, @n18, @n19], State=Unhovered)
   > StopAnimation(InAnimation=Get IsHoveredIncrease())
   n21 = Event SetNestedButtonGroupSelectionVM!() -> (OutputDelegate, then, InVM:Fort Nested Button Group Selection VM Object Reference) {is=title K2Node_Event}
   > Set FortNestedButtonGroupSelectionVM(FortNestedButtonGroupSelectionVM:Fort Nested Button Group Selection VM Object Reference=@n21.InVM) -> (Output_Get:Fort Nested Button Group Selection VM Object Reference)
   Event On Clicked()
   > n22 = Branch(Condition=Get bIsInInterestsChooserRow(self=Get FortDiscoverSurfaceTileItemVM()))
   > Branch(Condition=Get bEnableFortniteInterests!(self:Fort Runtime Options Object Reference=GetRuntimeOptions()) -> (bEnableFortniteInterests:bool))
   > TogglePendingFavorite(self=Get FortDiscoverSurfaceTileItemVM())
   > n23 = Branch(Condition=Get bIsPendingFavorite(self=Get FortDiscoverSurfaceTileItemVM()))
   > n24 = PlaySound2D(Sound=Get Favorited Sound())
   HandleCategoryTileClicked(
    execute=[@n24.then, @n25.then]
    self=Get FortActivitySelector()
    FortPC=GetFirstLocalFortPlayerController(bIncludeNonSpectators=true, bIncludeSpectators=false)
    InTileItemVM=Get FortDiscoverSurfaceTileItemVM()
   )
   n25 = PlaySound2D(execute=@n23.else, Sound=Get Unfavorited Sound())
   n26 = Parent: On Clicked(execute=@n22.else) {is=title K2Node_CallParentFunction}
   n27 = Event OnMouseEnter()
   > Parent: OnMouseEnter!(execute, MyGeometry:Geometry Structure=@n27.MyGeometry, MouseEvent:"Pointer Event Structure (by ref)"=@n27.MouseEvent) -> (then) {is=title K2Node_CallParentFunction}
   n28 = ScreenToLocal!(Geometry:"Geometry Structure (by ref)"=@n27.MyGeometry, ScreenPosition:Vector 2D Structure, bIncludeWindowPosition:bool=false) -> (LocalCoordinate_X:double, LocalCoordinate_Y:double)
   n29 = Get ImageMID(self=Get Foreground())
   > SetScalarParameterValue(self=@n29, ParameterName=ScreenPosX, Value=@n28)
   > SetScalarParameterValue(self=@n29, ParameterName=ScreenPosY, Value=@n28.LocalCoordinate_Y)
   n30 = Event OnListItemObjectSet()
   > n31 = Cast To FortDiscoverSurfaceTileItemVM!(execute, Object=@n30.ListItemObject) -> (then, CastFailed, AsFort Discover Surface Tile Item VM:Fort Discover Surface Tile Item VM Object Reference) {is=title K2Node_DynamicCast}
   > n32 = Set FortDiscoverSurfaceTileItemVM(FortDiscoverSurfaceTileItemVM:Fort Discover Surface Tile Item VM Object Reference=@n31) -> (Output_Get:Fort Discover Surface Tile Item VM Object Reference)
   > n33 = Branch(Condition=HasFetchedImages!(self:Fort Discover Link Entry Tile Item VM Object Reference=@n32) -> (ReturnValue:bool))
   ConfigureDebugUI!(
    execute=[@n34.then, @n33.then]
    self:WBP Discover Tile Base Object Reference=Get WBP_Discover_Tile_Base()
    Debug VM:Discover Tile Item Debug View Model Object Reference=Get DebugVM!(self:Discover Tile Item View Model Object Reference=Get FortDiscoverSurfaceTileItemVM()) -> (DebugVM:Discover Tile Item Debug View Model Object Reference)
    OwningButton:Common Button Base Object Reference=Self-Reference()
   ) -> (then)
   > SetTileItemViewModel!(
    execute
    self:WBP Discover Tile Base Object Reference=Get WBP_Discover_Tile_Base()
    TileItemViewModel:Discover Tile Item View Model Object Reference=@n32
   ) -> (then)
   n34 = FetchImages(execute=@n33.else, self:Fort Discover Link Entry Tile Item VM Object Reference=@n32)
  ]}
 GetBackgroundMaterialInstance {nodes [
   n1 = GetBackgroundMaterialInstance!() -> (then, GameCollectionId:EFortActivityLinkCategory Enum=None) {is=title K2Node_FunctionEntry}
   > Return Node!(
    execute
    MaterialInstance:Material Instance Object Reference=Select!(
     None:Material Instance Object Reference
     LEGO:Material Instance Object Reference=/DiscoveryBrowser/Tiles/Ip/MaterialInstances/MI_UI_Tile_IP_Background_Lego_B
     RocketRacing:Material Instance Object Reference=/DiscoveryBrowser/Tiles/Ip/MaterialInstances/MI_UI_Tile_IP_Background_RR
     FallGuys:Material Instance Object Reference=/DiscoveryBrowser/Tiles/Ip/MaterialInstances/MI_UI_Tile_IP_Background_FallGuys
     TMNT:Material Instance Object Reference=/DiscoveryBrowser/Tiles/Ip/MaterialInstances/MI_UI_Tile_IP_Background_TMNT
     TheWalkingDeadUniverse:Material Instance Object Reference=/DiscoveryBrowser/Tiles/Ip/MaterialInstances/MI_UI_Tile_IP_Background_DarkVortex
     SolidWave:Material Instance Object Reference=/DiscoveryBrowser/Tiles/Ip/MaterialInstances/MI_UI_Tile_IP_Background_SolidWave
     Akita:Material Instance Object Reference=/DiscoveryBrowser/Tiles/Ip/MaterialInstances/MI_UI_Tile_IP_Background_Akita
     KPOPDemonHunters:Material Instance Object Reference=/DiscoveryBrowser/Tiles/Ip/MaterialInstances/MI_UI_Tile_IP_Background_Refrigerator
     EpicGames:Material Instance Object Reference=/DiscoveryBrowser/Tiles/Ip/MaterialInstances/MI_UI_Tile_IP_Background_EpicGames
     Creatures:Material Instance Object Reference=/DiscoveryBrowser/Tiles/Ip/MaterialInstances/MI_UI_Tile_IP_Background_Creatures
     Index:EFortActivityLinkCategory Enum=@n1
    ) -> (ReturnValue:Material Instance Object Reference)
   )
  ]}
 GetFavoriteStatus {nodes [
   GetFavoriteStatus!() -> (then, IsPendingFavorite:bool=~, IsFavorite:bool=~, IsInInterestsRow:bool=~) {is=title K2Node_FunctionEntry}
   > n1 = Branch(Condition=Get bEnableFortniteInterests(self=GetRuntimeOptions()))
   > n2 = Branch(Condition=Get bIsInInterestsChooserRow(self=Get FortDiscoverSurfaceTileItemVM()))
   n3 = Select!(
    Option 0:bool=Get IsFavorite() -> (IsFavorite:bool)
    Option 1:bool=Get IsPendingFavorite() -> (IsPendingFavorite:bool)
    Index:bool=Get IsInInterestsRow() -> (IsInInterestsRow:bool)
   ) -> (ReturnValue:bool)
   Return Node!(
    execute=[@n5.then, @n2.then]
    Visibility:ESlateVisibility Enum=Select(Option 0:ESlateVisibility Enum=Collapsed, Option 1:ESlateVisibility Enum=SelfHitTestInvisible, Index=@n3) -> (ReturnValue:ESlateVisibility Enum)
   )
   n4 = Get WBP_Discover_Tile_Base()
   RemoveInteractionByName(execute=@n2.else, self=@n4, ActionToRemove=Get Action to Remove_Favorite() -> (Action to Remove_Favorite:Name=None))
   > n5 = AddInteraction(
    self=Get ContextMenuActionContext(self=@n4)
    InActionInfo_DisplayName=Select(Option 0:Text="LOCTEXT(\"Favorite\")", Option 1:Text="LOCTEXT(\"Unfavorite\")", Index=@n3) -> (ReturnValue:Text)
    InActionInfo_SortPriority=999
    InActionInfo_bIsAvailable=True
    InActionInfo_bIsEnabled=True
    InActionInfo_bIsSelected=False
    InActionInfo_bAddDivider=False
    InActionInfo_bPositionDividerAfter=False
    InActionInfo_bCloseMenuOnSelect=False
    InActionInfo_InteractionNameID=Get Action to Remove_Favorite() -> (Action to Remove_Favorite=None)
    InActionInfo_Interaction=(DataTable=None,RowName="")
    InActionInfo_PressedDelegate="(null).None"
    InActionInfo_ReleasedDelegate="(null).None"
    InInteractionDelegate=Create Event()
   )
   Return Node(execute=@n1.else, Visibility=Collapsed)
  ]}
 GetLogoTexture {nodes [
   n1 = GetLogoTexture!() -> (then, LogoLoadingState:EActivityImageLoadingState Enum=Unset, LogoTexture:Texture Object Reference) {is=title K2Node_FunctionEntry}
   > n2 = Switch on EActivityImageLoadingState!(execute, Selection:EActivityImageLoadingState Enum=@n1) -> (Unset:Exec, Loading:Exec, Loaded:Exec, Error:Exec) {is=title K2Node_SwitchEnum}
   Return Node!(execute=[@n2.Error, @n2.Loading, @n2.Unset], Texture:Texture Object Reference=/FortUILibrary/Textures/Icons/T_UI_Icon_UIKit_ImageError)
   Return Node(execute=@n2.Loaded, Texture=@n1.LogoTexture)
  ]}
 GetNestedButtonGroupSelectionVM {nodes [
   GetNestedButtonGroupSelectionVM() {is=title K2Node_FunctionEntry}
   > Return Node!(
    execute
    ReturnValue:Fort Nested Button Group Selection VM Object Reference=Get FortNestedButtonGroupSelectionVM() -> (FortNestedButtonGroupSelectionVM:Fort Nested Button Group Selection VM Object Reference)
   )
  ]}
 GetPopOutTexture {nodes [
   n1 = GetPopOutTexture!() -> (then, PopOutLoadingState:EActivityImageLoadingState Enum=Unset, PopOutTexture:Texture Object Reference) {is=title K2Node_FunctionEntry}
   > n2 = Switch on EActivityImageLoadingState(Selection=@n1)
   Return Node!(execute=[@n2.Error, @n2.Loading, @n2.Unset], Texture:Texture Object Reference=/Game/UI/DiscoveryShared/Art/T_UI_Discovery_DefaultEmpty)
   Return Node(execute=@n2.Loaded, Texture=@n1.PopOutTexture)
  ]}
 RefreshFocusVisual {nodes [
   RefreshFocusVisual() {is=title K2Node_FunctionEntry}
   > n1 = Branch(Condition=HasAnyUserFocus!(self:Widget Object Reference) -> (ReturnValue:bool))
   > n2 = Switch on ECommonInputType!(execute, Selection:ECommonInputType Enum=GetCurrentInputType(self=Get CommonInputSubsystem())) -> (MouseAndKeyboard:Exec, Gamepad:Exec, Touch:Exec, Count:Exec) {is=title K2Node_SwitchEnum}
   RequestTransition(execute=@n2.Gamepad, self=Get Foreground(), State=Focused)
   RequestTransition(execute=[@n2.MouseAndKeyboard, @n2.Touch, @n1.else], self=Get Foreground(), State=Unfocused)
  ]}
 SetBackgroundMaterialInstance {nodes [
   n1 = SetBackgroundMaterialInstance!() -> (then, BackgroundMaterialInstance:Material Instance Object Reference) {is=title K2Node_FunctionEntry}
   > SetBackgroundMaterial(self=Get Block_Background(), Material Instance=@n1)
  ]}
 SetCompactUIStyle {nodes [
   SetCompactUIStyle() {is=title K2Node_FunctionEntry}
   > Branch(Condition=IsMobileGame())
   > Set Tile Size(self=Get WBP_Discover_Tile_Base(), InSize=(X=480,Y=176))
   n1 = Get Logo()
   > Set Cap Height!(execute, Cap Height:int=96, self:WBP Home Bar Logo Object Reference=@n1) -> (then, Output_Get:int=0)
   > Set Cap Width!(execute, Cap Width:int=176, self:WBP Home Bar Logo Object Reference=@n1) -> (then, Output_Get:int=0)
   > Set SquareSize!(execute, SquareSize:int=112, self:WBP Home Bar Logo Object Reference=@n1) -> (then, Output_Get:int=0)
  ]}
 SetForegroundTexture {nodes [
   n1 = SetForegroundTexture!() -> (then, Texture:Texture Object Reference) {is=title K2Node_FunctionEntry}
   > SetTextureParameterValue(self=Get ImageMID(self=Get Foreground()), ParameterName=Texture, Value=@n1)
  ]}
 SetHoveredAdditive {nodes [
   n1 = SetHoveredAdditive!() -> (then, NewParam:double=~) {is=title K2Node_FunctionEntry}
   > n2 = Set HoveredAdditive(HoveredAdditive=@n1)
   > SetScalarParameterValue(
    self=Get BackgroundDynamicMaterial!(self:WBP UIKit Block Background Base Object Reference=Get Block_Background()) -> (BackgroundDynamicMaterial:Material Instance Dynamic Object Reference)
    ParameterName=Is Hovered Amount
    Value=float + float(A=@n2, B=@n1)
   )
  ]}
 SetLogoTexture {nodes [
   n1 = SetLogoTexture!() -> (then, Texture:Texture Object Reference) {is=title K2Node_FunctionEntry}
   > SetTexture(
    self=Get Logo()
    TextureSize=GetLogoTextureSize!(self:Fort Discover Link Entry Tile Item VM Object Reference=Get FortDiscoverSurfaceTileItemVM()) -> (ReturnValue:Vector 2D Structure)
    Texture=@n1
   )
  ]}
]
variables [
 Action to Remove_Collection {type=string, default=ViewCollection}
 Action to Remove_Favorite {type=string, default=ToggleFavorite}
 DebugBackgroundMaterialInstance {type=MaterialInstance, default=&MI_UI_Tile_IP_Background_FallGuys}
 DebugForegroundTexture {type=Texture2D, default=/DiscoveryBrowser/Tiles/Ip/Textures/T_UI_Discovery_Tile_Foreground_FallGuys}
 DebugLogoTexture {type=Texture2D, default=/DiscoveryBrowser/Tiles/Ip/Textures/T_UI_Discovery_Tile_Logo_FallGuys}
 Favorited Sound {type=SoundBase, default=/Game/Sounds/Fort_UI_Sounds/FrontEnd/MultiProduct/Common/Buttons/MSS_UI_MP_Button_Click_Favorite}
 ForegroundMaterialInstance {type=MaterialInstance, default=/DiscoveryBrowser/Tiles/Ip/MaterialInstances/MI_UI_Tile_IP_Foreground}
 HoveredAdditive {type=number}
 Unfavorited Sound {type=SoundBase, default=/Game/Sounds/Fort_UI_Sounds/FrontEnd/MultiProduct/Common/Buttons/MSS_UI_MP_Button_Click_Unfavorite}
]
widgets {
 info {
  inheritedWidgetCount=0
  namedSlotCount=3
  parentClass=/DiscoveryBrowser/Tiles/WBP_Discover_Tile_Button.WBP_Discover_Tile_Button_C
  rootWidgetClass=null
  widgetCount=7
 }
 widgets [
  GP_Content {widgetClassPath=&GridPanel, widget=GP_Content, properties {columnFill [1], rowFill [1]}}
  Block_Background {
   widgetClassPath=/FortUILibrary/Systems/ModularButton/Blocks/WBP_UIKit_Block_BackgroundGeneric
   bIsVariable=true
   parent=GP_Content
   widget=Block_Background
   properties {
    backgroundMaterial=&MI_UI_Tile_IP_Background_FallGuys
    transitionTiming {
     deselectingFocused=0
     deselectingIdle=0
     disabling=0
     enabling=0
     focusing=0.2
     hovering=0.2
     pressing=0
     releasing=0
     selecting=0
     unfocusing=0.2
     unhovering=0.2
    }
   }
   slot {class=GridSlot}
  }
  GridPanel_84 {widgetClassPath=&GridPanel, widget=GridPanel_84, properties {columnFill [1], rowFill [1]}}
  GridPanel_80 {widgetClassPath=&GridPanel, widget=GridPanel_80, properties {columnFill [1], rowFill [1]}}
  Foreground {
   widgetClassPath=/FortUILibrary/Systems/ModularButton/Blocks/WBP_UIKit_Block_Image
   bIsVariable=true
   parent=GridPanel_80
   widget=Foreground
   properties {
    pixelSnapping=Disabled
    transitionTiming {
     deselectingFocused=0
     deselectingIdle=0
     disabling=0
     enabling=0
     focusing=0.2
     hovering=0.2
     pressing=0
     releasing=0
     selecting=0
     unfocusing=0.2
     unhovering=0.2
    }
   }
   slot {class=GridSlot, properties {padding=<ltrb 0 -12 0 0>}}
  }
  Logo {
   widgetClassPath=/DiscoveryBrowser/TabScreens/DiscoverHome/WBP_HomeBar_Logo
   bIsVariable=true
   parent=GridPanel_80
   widget=Logo
   properties {
    cap Height=80
    cap Width=144
    squareSize=88
    texture=&T_UI_Discovery_Tile_Logo_FallGuys
    visibility=HitTestInvisible
   }
   slot {class=GridSlot, properties {horizontalAlignment=HAlign_Left, layer=1, verticalAlignment=VAlign_Center, padding=<ltrb 24 0 0 0>}}
  }
  StatusIndicator_Favorite {
   widgetClassPath=/FortUILibrary/UIKit/StatusIndicator/WBP_UIKit_StatusIndicator
   bIsVariable=true
   parent=GridPanel_80
   widget=StatusIndicator_Favorite
   properties {
    type=Utility
    icon=/DiscoveryBrowser/Tiles/Islands/Materials/MI_UIKit_Icon_Favorite_Discover
    iconSize=Compact
    overrideStatusIndicatorVisibility=true
    visibility=Collapsed
   }
   slot {class=GridSlot, properties {horizontalAlignment=HAlign_Left, layer=5, verticalAlignment=VAlign_Bottom, padding=<ltrb 8 0 0 8>}}
  }
 ]
}
